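# Control Rig for the male medium Raptor Knight body that drives the front loincloth bones from the thighs.
# Offset-preserving transform constraints bind the left chain to the left thigh twist, blend the right with its thigh twist and own pose, and blend the center between left thigh twist and right cloth.
# Shoulder pad and clavicle correctives, using multi-target constraints and angle-remapped offsets, are present but unconnected to Begin Execution.
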ControlRigBlueprint /BRCosmetics/Characters/Player/Male/Medium/Bodies/M_MED_Raptor_Knight/Meshes/M_MED_Raptor_Knight_CTRL_Rig
graphs [
 RigVMFunctionLibrary {links [], nodes []}
 RigVMModel {nodes [
   n1 = &RigUnit_GetRelativeTransformForItem(Child=(Type=Bone,Name="upperarm_r"), Parent=(Type=Bone,Name="spine_05"))
   &RigUnit_ModifyTransforms(
    ExecutePin=()
    ItemToModify=((Item=(Type=Bone,Name="dyn_r_shoulderPad"),Transform=(Rotation=(X=0,Y=0,Z=0,W=1),Translation=(X=50,Y=0,Z=0),Scale3D=(X=1,Y=1,Z=1))))
    Weight=&RigVMFunction_MathFloatRemap(
     SourceMaximum=87
     SourceMinimum=30
     TargetMinimum=0
     bClamp=true
     Value=&RigVMFunction_MathQuaternionToEuler(RotationOrder=ZYX, Value=@n1.RelativeTransform.Rotation).Result.Z
    ).Result
    ItemToModify.0.Transform.Rotation=&RigVMFunction_MathQuaternionFromEuler(Euler=(X=0,Y=0,Z=0)).Result
   )
   n2 = /Script/RigVM.RigVMFunction_MathFloatAbs(Value=&RigVMFunction_MathQuaternionToEuler(
     RotationOrder=ZYX
     Value=&RigUnit_GetRelativeTransformForItem(Child=(Type=Bone,Name="upperarm_l"), Parent=(Type=Bone,Name="spine_05")).RelativeTransform.Rotation
    ).Result.Z) {is=struct RigVMUnitNode, template="Absolute::Execute(in Value,out Result)"}
   > &RigUnit_ModifyTransforms(
    ItemToModify=((Item=(Type=Bone,Name="dyn_l_shoulderPad"),Transform=(Rotation=(X=0,Y=0,Z=0,W=1),Translation=(X=0,Y=-2,Z=2),Scale3D=(X=1,Y=1,Z=1))))
    Weight=&RigVMFunction_MathFloatRemap(SourceMaximum=69, SourceMinimum=100, TargetMinimum=0, bClamp=true, Value=@n2.Result).Result
    ItemToModify.0.Transform.Rotation=&RigVMFunction_MathQuaternionFromEuler(Euler=(X=30,Y=20,Z=0)).Result
   )
   n3 = &RigVMFunction_MathFloatAbs(Value=&RigVMFunction_MathQuaternionToEuler(
     RotationOrder=ZYX
     Value=&RigUnit_GetRelativeTransformForItem(Child=(Type=Bone,Name="upperarm_l"), Parent=(Type=Bone,Name="spine_05")).RelativeTransform.Rotation
    ).Result.Z) {template="Absolute::Execute(in Value,out Result)"}
   &RigUnit_ModifyTransforms(
    ExecutePin=()
    ItemToModify=((Item=(Type=Bone,Name="dyn_l_shoulderPad"),Transform=(Rotation=(X=0,Y=0,Z=0,W=1),Translation=(X=0,Y=-2.2,Z=1.2),Scale3D=(X=1,Y=1,Z=1))))
    Weight=&RigVMFunction_MathFloatRemap(SourceMaximum=69, SourceMinimum=100, TargetMinimum=0, bClamp=true, Value=@n3.Result).Result
    ItemToModify.0.Transform.Rotation=&RigVMFunction_MathQuaternionFromEuler(Euler=(X=0,Y=0,Z=0)).Result
   )
   > &RigUnit_ModifyTransforms(
    ItemToModify=((Item=(Type=Bone,Name="dyn_l_shoulderPad"),Transform=(Rotation=(X=0,Y=0,Z=0,W=1),Translation=(X=0,Y=0,Z=0.5),Scale3D=(X=1,Y=1,Z=1))))
    Weight=1
   )
   > &RigUnit_ModifyTransforms(
    ItemToModify=((Item=(Type=Bone,Name="upperarm_twist_01_l"),Transform=(Rotation=(X=0,Y=0,Z=0,W=1),Translation=(X=0,Y=0,Z=2),Scale3D=(X=1,Y=1,Z=1))))
    WeightMinimum=0
    Weight=&RigVMFunction_MathFloatRemap(SourceMaximum=62, SourceMinimum=92, TargetMinimum=0, bClamp=true, Value=@n3.Result).Result
   )
   n4 = &RigVMFunction_MathFloatAbs(Value=&RigVMFunction_MathQuaternionToEuler(
     RotationOrder=ZYX
     Value=&RigUnit_GetRelativeTransformForItem(Child=(Type=Bone,Name="clavicle_l"), Parent=(Type=Bone,Name="spine_05")).RelativeTransform.Rotation
    ).Result.Z) {template="Absolute::Execute(in Value,out Result)"}
   &RigUnit_ModifyTransforms(
    ExecutePin=()
    ItemToModify=((Item=(Type=Bone,Name="dyn_l_clavicle"),Transform=(Rotation=(X=0,Y=0,Z=0,W=1),Translation=(X=0,Y=3,Z=0),Scale3D=(X=1,Y=1,Z=1))))
    Weight=1
    ItemToModify.0.Transform.Rotation=&RigVMFunction_MathQuaternionFromEuler(
     Euler=(X=0,Y=0,Z=0)
     Euler.Z=&RigVMFunction_MathFloatRemap(SourceMaximum=50, SourceMinimum=100, TargetMinimum=0, bClamp=true, Value=@n4.Result).Result
    ).Result
   )
   > &RigUnit_ModifyTransforms(
    ItemToModify=((Item=(Type=Bone,Name="upperarm_twist_01_l"),Transform=(Rotation=(X=0,Y=0,Z=0,W=1),Translation=(X=0,Y=0,Z=2),Scale3D=(X=1,Y=1,Z=1))))
    WeightMinimum=0
    Weight=&RigVMFunction_MathFloatRemap(SourceMaximum=62, SourceMinimum=92, TargetMinimum=0, bClamp=true, Value=@n4.Result).Result
   )
   n5 = &RigVMFunction_MathFloatAbs(Value=&RigVMFunction_MathQuaternionToEuler(
     RotationOrder=ZYX
     Value=&RigUnit_GetRelativeTransformForItem(Child=(Type=Bone,Name="upperarm_l"), Parent=(Type=Bone,Name="spine_05")).RelativeTransform.Rotation
    ).Result.Z) {template="Absolute::Execute(in Value,out Result)"}
   &RigUnit_ModifyTransforms(
    ExecutePin=()
    ItemToModify=((Item=(Type=Bone,Name="dyn_l_clavicle"),Transform=(Rotation=(X=0,Y=0,Z=0,W=1),Translation=(X=0,Y=0,Z=0),Scale3D=(X=1,Y=1,Z=1))))
    Weight=&RigVMFunction_MathFloatRemap(SourceMaximum=69, SourceMinimum=100, TargetMinimum=0, bClamp=true, Value=@n5.Result).Result
    ItemToModify.0.Transform.Rotation=&RigVMFunction_MathQuaternionFromEuler(Euler=(X=0,Y=-5,Z=0)).Result
   )
   n6 = &RigUnit_GetRelativeTransformForItem(Child=(Type=Bone,Name="clavicle_r"), Parent=(Type=Bone,Name="spine_05"))
   > &RigUnit_ModifyTransforms(
    ItemToModify=((Item=(Type=Bone,Name="dyn_r_shoulderPad"),Transform=(Rotation=(X=0,Y=0,Z=0,W=1),Translation=(X=0,Y=-0.5,Z=-1),Scale3D=(X=1,Y=1,Z=1))))
    WeightMinimum=0
    Weight=&RigVMFunction_MathFloatRemap(
     SourceMaximum=-90
     SourceMinimum=0
     TargetMinimum=0
     bClamp=true
     Value=&RigVMFunction_MathQuaternionToEuler(RotationOrder=ZYX, Value=@n6.RelativeTransform.Rotation).Result.Z
    ).Result
    ItemToModify.0.Transform.Rotation=&RigVMFunction_MathQuaternionFromEuler(Euler=(X=5,Y=0,Z=0)).Result
   )
   n7 = &RigVMFunction_MathFloatAbs(Value=&RigVMFunction_MathQuaternionToEuler(
     RotationOrder=ZYX
     Value=&RigUnit_GetRelativeTransformForItem(Child=(Type=Bone,Name="clavicle_l"), Parent=(Type=Bone,Name="spine_05")).RelativeTransform.Rotation
    ).Result.Z) {template="Absolute::Execute(in Value,out Result)"}
   > &RigUnit_ModifyTransforms(
    ItemToModify=((Item=(Type=Bone,Name="dyn_l_shoulderPad"),Transform=(Rotation=(X=0,Y=0,Z=0,W=1),Translation=(X=0,Y=0,Z=0),Scale3D=(X=1,Y=1,Z=1))))
    Weight=&RigVMFunction_MathFloatRemap(SourceMaximum=88, SourceMinimum=0, TargetMinimum=0, bClamp=true, Value=@n7.Result).Result
    ItemToModify.0.Transform.Rotation=&RigVMFunction_MathQuaternionFromEuler(Euler=(X=10,Y=0,Z=0)).Result
   )
   RigUnit_BeginExecution = &RigUnit_BeginExecution()
   /Script/ControlRig.RigUnit_TransformConstraintPerItem(
    BaseItem=(Type=Bone)
    BaseTransform=(Rotation=(X=0,Y=0,Z=0,W=1),Translation=(X=0,Y=0,Z=0),Scale3D=(X=1,Y=1,Z=1))
    BaseTransformSpace=GlobalSpace
    Item=(Type=Bone,Name="dyn_loinClothFront_l_MSTR")
    Targets=((Transform=(Rotation=(X=0,Y=0,Z=0,W=1),Translation=(X=0,Y=0,Z=0),Scale3D=(X=1,Y=1,Z=1)),Weight=1,bMaintainOffset=True,Filter=(TranslationFilter=(bX=True,bY=True,bZ=True),RotationFilter=(bX=True,bY=True,bZ=True),ScaleFilter=(bX=True,bY=True,bZ=True))))
    bUseInitialTransforms=True
    Targets.0.Transform=&RigUnit_GetTransform(Item=(Type=Bone,Name="thigh_twist_01_l")).Transform
    ExecutePin=@RigUnit_BeginExecution.ExecutePin
   ) {is=struct RigVMUnitNode}
   > &RigUnit_TransformConstraintPerItem(
    BaseItem=(Type=Bone)
    BaseTransform=(Rotation=(X=0,Y=0,Z=0,W=1),Translation=(X=0,Y=0,Z=0),Scale3D=(X=1,Y=1,Z=1))
    BaseTransformSpace=GlobalSpace
    Item=(Type=Bone,Name="dyn_loinClothFront_01_l")
    Targets=((Transform=(Rotation=(X=0,Y=0,Z=0,W=1),Translation=(X=0,Y=0,Z=0),Scale3D=(X=1,Y=1,Z=1)),Weight=1,bMaintainOffset=True,Filter=(TranslationFilter=(bX=True,bY=True,bZ=True),RotationFilter=(bX=True,bY=True,bZ=True),ScaleFilter=(bX=True,bY=True,bZ=True))))
    bUseInitialTransforms=True
    Targets.0.Transform=&RigUnit_GetTransform(Item=(Type=Bone,Name="thigh_twist_01_l")).Transform
   )
   > &RigUnit_TransformConstraintPerItem(
    BaseItem=(Type=Bone)
    BaseTransform=(Rotation=(X=0,Y=0,Z=0,W=1),Translation=(X=0,Y=0,Z=0),Scale3D=(X=1,Y=1,Z=1))
    BaseTransformSpace=GlobalSpace
    Item=(Type=Bone,Name="dyn_loinClothFront_02_l")
    Targets=((Transform=(Rotation=(X=0,Y=0,Z=0,W=1),Translation=(X=0,Y=0,Z=0),Scale3D=(X=1,Y=1,Z=1)),Weight=1,bMaintainOffset=True,Filter=(TranslationFilter=(bX=True,bY=True,bZ=True),RotationFilter=(bX=True,bY=True,bZ=True),ScaleFilter=(bX=True,bY=True,bZ=True))))
    bUseInitialTransforms=True
    Targets.0.Transform=&RigUnit_GetTransform(Item=(Type=Bone,Name="thigh_twist_01_l")).Transform
   )
   > &RigUnit_TransformConstraintPerItem(
    BaseItem=(Type=Bone)
    BaseTransform=(Rotation=(X=0,Y=0,Z=0,W=1),Translation=(X=0,Y=0,Z=0),Scale3D=(X=1,Y=1,Z=1))
    BaseTransformSpace=GlobalSpace
    Item=(Type=Bone,Name="dyn_loinClothFront_03_l")
    Targets=((Transform=(Rotation=(X=0,Y=0,Z=0,W=1),Translation=(X=0,Y=0,Z=0),Scale3D=(X=1,Y=1,Z=1)),Weight=1,bMaintainOffset=True,Filter=(TranslationFilter=(bX=True,bY=True,bZ=True),RotationFilter=(bX=True,bY=True,bZ=True),ScaleFilter=(bX=True,bY=True,bZ=True))))
    bUseInitialTransforms=True
    Targets.0.Transform=&RigUnit_GetTransform(Item=(Type=Bone,Name="thigh_twist_01_l")).Transform
   )
   > &RigUnit_TransformConstraintPerItem(
    BaseItem=(Type=Bone)
    BaseTransform=(Rotation=(X=0,Y=0,Z=0,W=1),Translation=(X=0,Y=0,Z=0),Scale3D=(X=1,Y=1,Z=1))
    BaseTransformSpace=GlobalSpace
    Item=(Type=Bone,Name="dyn_loinClothFront_r_MSTR")
    Targets="((Transform=(Rotation=(X=0,Y=0,Z=0,W=1),Translation=(X=0,Y=0,Z=0),Scale3D=(X=1,Y=1,Z=1)),Weight=0.5,bMaintainOffset=True,Filter=(TranslationFilter=(bX=True,bY=True,bZ=True),RotationFilter=(bX=True,bY=True,bZ=True),ScaleFilter=(bX=True,bY=True,bZ=True))),(T ...(truncated)"
    bUseInitialTransforms=True
    Targets.0.Transform=&RigUnit_GetTransform(Item=(Type=Bone,Name="thigh_twist_01_r")).Transform
    Targets.1.Transform=&RigUnit_GetTransform(Item=(Type=Bone,Name="dyn_loinClothFront_r_MSTR")).Transform
   )
   > &RigUnit_TransformConstraintPerItem(
    BaseItem=(Type=Bone)
    BaseTransform=(Rotation=(X=0,Y=0,Z=0,W=1),Translation=(X=0,Y=0,Z=0),Scale3D=(X=1,Y=1,Z=1))
    BaseTransformSpace=GlobalSpace
    Item=(Type=Bone,Name="dyn_loinClothFront_01_r")
    Targets="((Transform=(Rotation=(X=0,Y=0,Z=0,W=1),Translation=(X=0,Y=0,Z=0),Scale3D=(X=1,Y=1,Z=1)),Weight=0.75,bMaintainOffset=True,Filter=(TranslationFilter=(bX=True,bY=True,bZ=True),RotationFilter=(bX=True,bY=True,bZ=True),ScaleFilter=(bX=True,bY=True,bZ=True))),( ...(truncated)"
    bUseInitialTransforms=True
    Targets.0.Transform=&RigUnit_GetTransform(Item=(Type=Bone,Name="thigh_twist_01_r")).Transform
    Targets.1.Transform=&RigUnit_GetTransform(Item=(Type=Bone,Name="dyn_loinClothFront_01_r")).Transform
   )
   > &RigUnit_TransformConstraintPerItem(
    BaseItem=(Type=Bone)
    BaseTransform=(Rotation=(X=0,Y=0,Z=0,W=1),Translation=(X=0,Y=0,Z=0),Scale3D=(X=1,Y=1,Z=1))
    BaseTransformSpace=GlobalSpace
    Item=(Type=Bone,Name="dyn_loinClothFront_02_r")
    Targets="((Transform=(Rotation=(X=0,Y=0,Z=0,W=1),Translation=(X=0,Y=0,Z=0),Scale3D=(X=1,Y=1,Z=1)),Weight=0.75,bMaintainOffset=True,Filter=(TranslationFilter=(bX=True,bY=True,bZ=True),RotationFilter=(bX=True,bY=True,bZ=True),ScaleFilter=(bX=True,bY=True,bZ=True))),( ...(truncated)"
    bUseInitialTransforms=True
    Targets.0.Transform=&RigUnit_GetTransform(Item=(Type=Bone,Name="thigh_twist_01_r")).Transform
    Targets.1.Transform=&RigUnit_GetTransform(Item=(Type=Bone,Name="dyn_loinClothFront_02_r")).Transform
   )
   > &RigUnit_TransformConstraintPerItem(
    BaseItem=(Type=Bone)
    BaseTransform=(Rotation=(X=0,Y=0,Z=0,W=1),Translation=(X=0,Y=0,Z=0),Scale3D=(X=1,Y=1,Z=1))
    BaseTransformSpace=GlobalSpace
    Item=(Type=Bone,Name="dyn_loinClothFront_03_r")
    Targets="((Transform=(Rotation=(X=0,Y=0,Z=0,W=1),Translation=(X=0,Y=0,Z=0),Scale3D=(X=1,Y=1,Z=1)),Weight=0.5,bMaintainOffset=True,Filter=(TranslationFilter=(bX=True,bY=True,bZ=True),RotationFilter=(bX=True,bY=True,bZ=True),ScaleFilter=(bX=True,bY=True,bZ=True))),(T ...(truncated)"
    bUseInitialTransforms=True
    Targets.0.Transform=&RigUnit_GetTransform(Item=(Type=Bone,Name="thigh_twist_01_r")).Transform
    Targets.1.Transform=&RigUnit_GetTransform(Item=(Type=Bone,Name="dyn_loinClothFront_03_r")).Transform
   )
   > &RigUnit_TransformConstraintPerItem(
    BaseItem=(Type=Bone)
    BaseTransform=(Rotation=(X=0,Y=0,Z=0,W=1),Translation=(X=0,Y=0,Z=0),Scale3D=(X=1,Y=1,Z=1))
    BaseTransformSpace=GlobalSpace
    Item=(Type=Bone,Name="dyn_loinClothFront_c_MSTR")
    Targets="((Transform=(Rotation=(X=0,Y=0,Z=0,W=1),Translation=(X=0,Y=0,Z=0),Scale3D=(X=1,Y=1,Z=1)),Weight=0.5,bMaintainOffset=True,Filter=(TranslationFilter=(bX=True,bY=True,bZ=True),RotationFilter=(bX=True,bY=True,bZ=True),ScaleFilter=(bX=True,bY=True,bZ=True))),(T ...(truncated)"
    bUseInitialTransforms=True
    Targets.0.Transform=&RigUnit_GetTransform(Item=(Type=Bone,Name="thigh_twist_01_l")).Transform
    Targets.1.Transform=&RigUnit_GetTransform(Item=(Type=Bone,Name="dyn_loinClothFront_r_MSTR")).Transform
   )
   > &RigUnit_TransformConstraintPerItem(
    BaseItem=(Type=Bone)
    BaseTransform=(Rotation=(X=0,Y=0,Z=0,W=1),Translation=(X=0,Y=0,Z=0),Scale3D=(X=1,Y=1,Z=1))
    BaseTransformSpace=GlobalSpace
    Item=(Type=Bone,Name="dyn_loinClothFront_01_c")
    Targets="((Transform=(Rotation=(X=0,Y=0,Z=0,W=1),Translation=(X=0,Y=0,Z=0),Scale3D=(X=1,Y=1,Z=1)),Weight=0.5,bMaintainOffset=True,Filter=(TranslationFilter=(bX=True,bY=True,bZ=True),RotationFilter=(bX=True,bY=True,bZ=True),ScaleFilter=(bX=True,bY=True,bZ=True))),(T ...(truncated)"
    bUseInitialTransforms=True
    Targets.0.Transform=&RigUnit_GetTransform(Item=(Type=Bone,Name="thigh_twist_01_l")).Transform
    Targets.1.Transform=&RigUnit_GetTransform(Item=(Type=Bone,Name="dyn_loinClothFront_01_r")).Transform
   )
   > &RigUnit_TransformConstraintPerItem(
    BaseItem=(Type=Bone)
    BaseTransform=(Rotation=(X=0,Y=0,Z=0,W=1),Translation=(X=0,Y=0,Z=0),Scale3D=(X=1,Y=1,Z=1))
    BaseTransformSpace=GlobalSpace
    Item=(Type=Bone,Name="dyn_loinClothFront_02_c")
    Targets="((Transform=(Rotation=(X=0,Y=0,Z=0,W=1),Translation=(X=0,Y=0,Z=0),Scale3D=(X=1,Y=1,Z=1)),Weight=0.5,bMaintainOffset=True,Filter=(TranslationFilter=(bX=True,bY=True,bZ=True),RotationFilter=(bX=True,bY=True,bZ=True),ScaleFilter=(bX=True,bY=True,bZ=True))),(T ...(truncated)"
    bUseInitialTransforms=True
    Targets.0.Transform=&RigUnit_GetTransform(Item=(Type=Bone,Name="thigh_twist_01_l")).Transform
    Targets.1.Transform=&RigUnit_GetTransform(Item=(Type=Bone,Name="dyn_loinClothFront_02_r")).Transform
   )
   > &RigUnit_TransformConstraintPerItem(
    BaseItem=(Type=Bone)
    BaseTransform=(Rotation=(X=0,Y=0,Z=0,W=1),Translation=(X=0,Y=0,Z=0),Scale3D=(X=1,Y=1,Z=1))
    BaseTransformSpace=GlobalSpace
    Item=(Type=Bone,Name="dyn_loinClothFront_03_c")
    Targets="((Transform=(Rotation=(X=0,Y=0,Z=0,W=1),Translation=(X=0,Y=0,Z=0),Scale3D=(X=1,Y=1,Z=1)),Weight=0.5,bMaintainOffset=True,Filter=(TranslationFilter=(bX=True,bY=True,bZ=True),RotationFilter=(bX=True,bY=True,bZ=True),ScaleFilter=(bX=True,bY=True,bZ=True))),(T ...(truncated)"
    bUseInitialTransforms=True
    Targets.0.Transform=&RigUnit_GetTransform(Item=(Type=Bone,Name="thigh_twist_01_l")).Transform
    Targets.1.Transform=&RigUnit_GetTransform(Item=(Type=Bone,Name="dyn_loinClothFront_03_r")).Transform
   )
   &RigUnit_TransformConstraintPerItem(
    BaseItem=(Type=Bone)
    BaseTransform=(Rotation=(X=0,Y=0,Z=0,W=1),Translation=(X=0,Y=0,Z=0),Scale3D=(X=1,Y=1,Z=1))
    BaseTransformSpace=GlobalSpace
    ExecutePin=()
    Item=(Type=Bone,Name="dyn_r_clavicle")
    Targets="((Transform=(Rotation=(X=0,Y=0,Z=0,W=1),Translation=(X=0,Y=0,Z=0),Scale3D=(X=1,Y=1,Z=1)),Weight=0,bMaintainOffset=True,Filter=(TranslationFilter=(bX=True,bY=True,bZ=True),RotationFilter=(bX=True,bY=True,bZ=True),ScaleFilter=(bX=True,bY=True,bZ=True))),(Tra ...(truncated)"
    bUseInitialTransforms=True
    Targets.0.Transform=&RigUnit_GetTransform(Item=(Type=Bone,Name="spine_03")).Transform
    Targets.1.Transform=&RigUnit_GetTransform(Item=(Type=Bone,Name="spine_04")).Transform
    Targets.2.Transform=&RigUnit_GetTransform(Item=(Type=Bone,Name="spine_05")).Transform
    Targets.3.Transform=&RigUnit_GetTransform(Item=(Type=Bone,Name="clavicle_r")).Transform
    Targets.4.Transform=&RigUnit_GetTransform(Item=(Type=Bone,Name="upperarm_twist_01_r")).Transform
   )
   > &RigUnit_TransformConstraintPerItem(
    BaseItem=(Type=Bone)
    BaseTransform=(Rotation=(X=0,Y=0,Z=0,W=1),Translation=(X=0,Y=0,Z=0),Scale3D=(X=1,Y=1,Z=1))
    BaseTransformSpace=GlobalSpace
    Item=(Type=Bone,Name="dyn_l_clavicle")
    Targets="((Transform=(Rotation=(X=0,Y=0,Z=0,W=1),Translation=(X=0,Y=0,Z=0),Scale3D=(X=1,Y=1,Z=1)),Weight=0.2,bMaintainOffset=True,Filter=(TranslationFilter=(bX=True,bY=True,bZ=True),RotationFilter=(bX=True,bY=True,bZ=True),ScaleFilter=(bX=True,bY=True,bZ=True))),(T ...(truncated)"
    bUseInitialTransforms=True
    Targets.4.Transform=&RigUnit_GetTransform(Item=(Type=Bone,Name="upperarm_twist_01_l")).Transform
    Targets.3.Transform=&RigUnit_GetTransform(Item=(Type=Bone,Name="clavicle_l")).Transform
    Targets.2.Transform=&RigUnit_GetTransform(Item=(Type=Bone,Name="spine_05")).Transform
    Targets.1.Transform=&RigUnit_GetTransform(Item=(Type=Bone,Name="spine_04")).Transform
    Targets.0.Transform=&RigUnit_GetTransform(Item=(Type=Bone,Name="spine_03")).Transform
   )
   &RigUnit_TransformConstraintPerItem(
    BaseItem=(Type=Bone)
    BaseTransform=(Rotation=(X=0,Y=0,Z=0,W=1),Translation=(X=0,Y=0,Z=0),Scale3D=(X=1,Y=1,Z=1))
    BaseTransformSpace=GlobalSpace
    ExecutePin=()
    Item=(Type=Bone,Name="dyn_r_shoulderPad")
    Targets="((Transform=(Rotation=(X=0,Y=0,Z=0,W=1),Translation=(X=0,Y=0,Z=0),Scale3D=(X=1,Y=1,Z=1)),Weight=0,bMaintainOffset=True,Filter=(TranslationFilter=(bX=True,bY=True,bZ=True),RotationFilter=(bX=True,bY=True,bZ=True),ScaleFilter=(bX=True,bY=True,bZ=True))),(Tra ...(truncated)"
    bUseInitialTransforms=True
    Targets.4.Transform=&RigUnit_GetTransform(Item=(Type=Bone,Name="upperarm_twist_01_r")).Transform
    Targets.3.Transform=&RigUnit_GetTransform(Item=(Type=Bone,Name="clavicle_r")).Transform
    Targets.2.Transform=&RigUnit_GetTransform(Item=(Type=Bone,Name="spine_05")).Transform
    Targets.1.Transform=&RigUnit_GetTransform(Item=(Type=Bone,Name="spine_04")).Transform
    Targets.0.Transform=&RigUnit_GetTransform(Item=(Type=Bone,Name="spine_03")).Transform
   )
   > &RigUnit_TransformConstraintPerItem(
    BaseItem=(Type=Bone)
    BaseTransform=(Rotation=(X=0,Y=0,Z=0,W=1),Translation=(X=0,Y=0,Z=0),Scale3D=(X=1,Y=1,Z=1))
    BaseTransformSpace=GlobalSpace
    Item=(Type=Bone,Name="dyn_l_shoulderPad")
    Targets="((Transform=(Rotation=(X=0,Y=0,Z=0,W=1),Translation=(X=0,Y=0,Z=0),Scale3D=(X=1,Y=1,Z=1)),Weight=0,bMaintainOffset=True,Filter=(TranslationFilter=(bX=True,bY=True,bZ=True),RotationFilter=(bX=True,bY=True,bZ=True),ScaleFilter=(bX=True,bY=True,bZ=True))),(Tra ...(truncated)"
    bUseInitialTransforms=True
    Targets.4.Transform=&RigUnit_GetTransform(Item=(Type=Bone,Name="upperarm_twist_01_l")).Transform
    Targets.3.Transform=&RigUnit_GetTransform(Item=(Type=Bone,Name="clavicle_l")).Transform
    Targets.2.Transform=&RigUnit_GetTransform(Item=(Type=Bone,Name="spine_05")).Transform
    Targets.1.Transform=&RigUnit_GetTransform(Item=(Type=Bone,Name="spine_04")).Transform
    Targets.0.Transform=&RigUnit_GetTransform(Item=(Type=Bone,Name="spine_03")).Transform
   )
   &RigVMFunction_MathFloatRemap(SourceMaximum=62, SourceMinimum=92, TargetMinimum=0, bClamp=true, Value=@n2.Result)
   &RigVMFunction_MathFloatRemap(SourceMaximum=62, SourceMinimum=92, TargetMinimum=0, bClamp=true, Value=@n5.Result)
   n8 = &RigVMFunction_MathFloatAbs(Value=&RigVMFunction_MathQuaternionToEuler(
     RotationOrder=ZYX
     Value=&RigUnit_GetRelativeTransformForItem(Child=(Type=Bone,Name="clavicle_l"), Parent=(Type=Bone,Name="spine_05")).RelativeTransform.Rotation
    ).Result.Z) {template="Absolute::Execute(in Value,out Result)"}
   &RigVMFunction_MathFloatRemap(SourceMaximum=62, SourceMinimum=92, TargetMinimum=0, bClamp=true, Value=@n8.Result)
   &RigVMFunction_MathFloatRemap(SourceMaximum=62, SourceMinimum=92, TargetMinimum=0, bClamp=true, Value=@n7.Result)
   &RigUnit_ModifyTransforms(
    ExecutePin=()
    ItemToModify=((Item=(Type=Bone,Name="dyn_l_clavicle"),Transform=(Rotation=(X=0,Y=0,Z=0,W=1),Translation=(X=0,Y=0,Z=0),Scale3D=(X=1,Y=1,Z=1))))
    Weight=&RigVMFunction_MathFloatRemap(SourceMaximum=69, SourceMinimum=100, TargetMinimum=0, bClamp=true, Value=@n8.Result).Result
    ItemToModify.0.Transform.Rotation=&RigVMFunction_MathQuaternionFromEuler(Euler=(X=0,Y=0,Z=0)).Result
   )
   n9 = &RigUnit_GetRelativeTransformForItem(Child=(Type=Bone,Name="upperarm_r"), Parent=(Type=Bone,Name="spine_05"))
   &RigUnit_ModifyTransforms(
    ExecutePin=()
    ItemToModify=((Item=(Type=Bone,Name="dyn_r_clavicle"),Transform=(Rotation=(X=0,Y=0,Z=0,W=1),Translation=(X=0,Y=0,Z=0),Scale3D=(X=1,Y=1,Z=1))))
    Weight=&RigVMFunction_MathFloatRemap(
     SourceMaximum=87
     SourceMinimum=30
     TargetMinimum=0
     bClamp=true
     Value=&RigVMFunction_MathQuaternionToEuler(RotationOrder=ZYX, Value=@n9.RelativeTransform.Rotation).Result.Z
    ).Result
    ItemToModify.0.Transform.Rotation=&RigVMFunction_MathQuaternionFromEuler(Euler=(X=0,Y=-8,Z=5)).Result
   )
   &RigUnit_ModifyTransforms(
    ExecutePin=()
    ItemToModify=((Item=(Type=Bone,Name="dyn_l_clavicle"),Transform=(Rotation=(X=0,Y=0,Z=0,W=1),Translation=(X=0,Y=0,Z=0.5),Scale3D=(X=1,Y=1,Z=1))))
    Weight=1
   )
   n10 = &RigUnit_GetRelativeTransformForItem(Child=(Type=Bone,Name="clavicle_r"), Parent=(Type=Bone,Name="spine_05"))
   &RigUnit_ModifyTransforms(
    ExecutePin=()
    ItemToModify=((Item=(Type=Bone,Name="dyn_r_clavicle"),Transform=(Rotation=(X=0,Y=0,Z=0,W=1),Translation=(X=0,Y=0,Z=-1),Scale3D=(X=1,Y=1,Z=1))))
    Weight=&RigVMFunction_MathFloatRemap(
     SourceMaximum=-90
     SourceMinimum=0
     TargetMinimum=0
     bClamp=true
     Value=&RigVMFunction_MathQuaternionToEuler(RotationOrder=ZYX, Value=@n10.RelativeTransform.Rotation).Result.Z
    ).Result
    ItemToModify.0.Transform.Rotation=&RigVMFunction_MathQuaternionFromEuler(Euler=(X=0,Y=0,Z=-6)).Result
   )
  ]}
]
hierarchy [
 root {loc=<0 0 0>, rot=<0 0 0 1>}
 attach {parent=root, loc=<0 0 -0.7>, rot=<0 0 0.0529 -0.9986>}
 pelvis {parent=root, loc=<0 -0.7071 93.1191>, rot=<0.0127 0.707 -0.0127 -0.707>}
 spine_01 {parent=pelvis, loc=<6.7494 -0.6388 0>, rot=<0 0 0.034 -0.9994>}
 spine_02 {parent=spine_01, loc=<7.2518 -0.035 0>, rot=<0 0 0.0238 -0.9997>}
 spine_03 {parent=spine_02, loc=<6.0104 0 0>, rot=<0 0 0.0149 -0.9999>}
 spine_04 {parent=spine_03, loc=<8.1024 0 0>, rot=<0 0 -0.0132 -0.9999>}
 spine_05 {parent=spine_04, loc=<10.0146 0 0>, rot=<0 0 -0.0015 -1>}
 clavicle_l {parent=spine_05, loc=<4.3575 -0.1466 -2.3223>, rot=<0.0465 -0.7289 -0.009 -0.6829>}
 upperarm_l {parent=clavicle_l, loc=<15.5111 0 0>, rot=<0.0226 -0.3234 -0.0088 -0.9459>}
 lowerarm_l {parent=upperarm_l, loc=<26.6514 0 0>, rot=<0.0391 -0.0099 -0.2453 -0.9686>}
 hand_l {parent=lowerarm_l, loc=<22.3491 0 0>, rot=<-0.7081 -0.0507 0.0975 -0.6975>}
 index_metacarpal_l {parent=hand_l, loc=<1.6535 0.0183 -2.96>, rot=<0.0158 -0.0124 -0.0173 -0.9996>}
 index_01_l {parent=index_metacarpal_l, loc=<8.3008 0.0357 -0.0524>, rot=<-0.0044 0.0485 0.0924 -0.9945>}
 index_02_l {parent=index_01_l, loc=<3.3587 -0.3004 0.0452>, rot=<-0.002 0.0027 0.3225 -0.9466>}
 index_03_l {parent=index_02_l, loc=<2.5578 -0.123 -0.0252>, rot=<-0.002 0.0067 0.2558 -0.9667>}
 middle_metacarpal_l {parent=hand_l, loc=<2.1856 -0.1725 -0.5984>, rot=<0.006 0.0168 -0.0329 -0.9993>}
 middle_01_l {parent=middle_metacarpal_l, loc=<7.3275 -0.2447 0.0129>, rot=<-0.0208 0.0042 0.0858 -0.9961>}
 middle_02_l {parent=middle_01_l, loc=<4.0123 0.1401 0.1175>, rot=<-0.0218 0.0258 0.3512 -0.9357>}
 middle_03_l {parent=middle_02_l, loc=<2.6077 0.0534 0.1373>, rot=<0.0082 0.0188 0.1708 -0.9851>}
 pinky_metacarpal_l {parent=hand_l, loc=<1.7616 -0.9753 3.3814>, rot=<-0.0397 0.0814 -0.069 -0.9935>}
 pinky_01_l {parent=pinky_metacarpal_l, loc=<7.1709 -0.4831 -0.1094>, rot=<-0.0654 -0.0035 0.2242 -0.9723>}
 pinky_02_l {parent=pinky_01_l, loc=<2.8274 0.0893 -0.1043>, rot=<0.0037 0.0011 0.2854 -0.9584>}
 pinky_03_l {parent=pinky_02_l, loc=<2.2046 0.0307 -0.0078>, rot=<-0.0002 0.0098 0.1445 -0.9895>}
 ring_metacarpal_l {parent=hand_l, loc=<2.2074 -0.3625 1.3638>, rot=<-0.0284 0.0722 -0.0458 -0.9959>}
 ring_01_l {parent=ring_metacarpal_l, loc=<7.0295 -0.6197 -0.1446>, rot=<-0.0489 -0.0252 0.1892 -0.9804>}
 ring_02_l {parent=ring_01_l, loc=<3.6174 0.4647 -0.1027>, rot=<0 0 0.2806 -0.9598>}
 ring_03_l {parent=ring_02_l, loc=<2.4152 0.114 0.0199>, rot=<0 0 0.1189 -0.9929>}
 thumb_01_l {parent=hand_l, loc=<1.3118 -0.3191 -2.459>, rot=<0.6546 -0.4042 0.0664 -0.6354>}
 thumb_02_l {parent=thumb_01_l, loc=<5.9336 -0.0448 0.0401>, rot=<-0.0037 0.0213 0.298 -0.9543>}
 thumb_03_l {parent=thumb_02_l, loc=<2.2223 0.3682 -0.0907>, rot=<0.0118 -0.013 0.0066 -0.9998>}
 weapon_l {parent=hand_l, loc=<-0.1974 -3.4111 0.3811>, rot=<-0.0006 0.0002 0.0002 -1>}
 lowerarm_twist_01_l {parent=lowerarm_l, loc=<17.8549 0 0>, rot=<0 0 0 -1>}
 lowerarm_twist_02_l {parent=lowerarm_l, loc=<9.6199 0 0>, rot=<0 0 0 -1>}
 upperarm_twist_01_l {parent=upperarm_l, loc=<0.0041 0 0>, rot=<0 0 0 -1>}
 upperarm_twist_02_l {parent=upperarm_l, loc=<13.7941 0 0>, rot=<0 0 0 -1>}
 dyn_l_shoulderPad {parent=upperarm_l, loc=<0 0 0>, rot=<0 0 0 -1>}
 dyn_l_clavicle {parent=clavicle_l, loc=<0 0 0>, rot=<0 0 0 1>}
 clavicle_r {parent=spine_05, loc=<4.3575 -0.1466 2.3223>, rot=<0.7289 0.0465 -0.6829 0.009>}
 upperarm_r {parent=clavicle_r, loc=<-15.5111 0 0>, rot=<0.0226 -0.3234 -0.0088 -0.9459>}
 lowerarm_r {parent=upperarm_r, loc=<-26.6515 0 0>, rot=<0.0391 -0.0099 -0.2453 -0.9686>}
 hand_r {parent=lowerarm_r, loc=<-22.3492 0 0>, rot=<-0.7081 -0.0507 0.0975 -0.6975>}
 index_metacarpal_r {parent=hand_r, loc=<-1.6534 -0.0181 2.9601>, rot=<0.0158 -0.0124 -0.0173 -0.9996>}
 index_01_r {parent=index_metacarpal_r, loc=<-8.3008 -0.0357 0.0524>, rot=<-0.0044 0.0485 0.0924 -0.9945>}
 index_02_r {parent=index_01_r, loc=<-3.3589 0.3003 -0.0453>, rot=<-0.002 0.0027 0.3225 -0.9466>}
 index_03_r {parent=index_02_r, loc=<-2.5578 0.1231 0.0251>, rot=<-0.002 0.0067 0.2558 -0.9667>}
 middle_metacarpal_r {parent=hand_r, loc=<-2.1856 0.1726 0.5984>, rot=<0.006 0.0168 -0.0329 -0.9993>}
 middle_01_r {parent=middle_metacarpal_r, loc=<-7.3273 0.2447 -0.0129>, rot=<-0.0208 0.0042 0.0858 -0.9961>}
 middle_02_r {parent=middle_01_r, loc=<-4.0142 -0.1411 -0.07>, rot=<-0.0217 0.0259 0.3512 -0.9357>}
 middle_03_r {parent=middle_02_r, loc=<-2.6074 -0.0534 -0.1312>, rot=<0.0083 0.0188 0.1708 -0.9851>}
 pinky_metacarpal_r {parent=hand_r, loc=<-1.7617 0.9753 -3.3814>, rot=<-0.0397 0.0814 -0.069 -0.9935>}
 pinky_01_r {parent=pinky_metacarpal_r, loc=<-7.1707 0.4832 0.1094>, rot=<-0.0654 -0.0035 0.2242 -0.9723>}
 pinky_02_r {parent=pinky_01_r, loc=<-2.8274 -0.0892 0.1043>, rot=<0.0037 0.0011 0.2854 -0.9584>}
 pinky_03_r {parent=pinky_02_r, loc=<-2.2045 -0.031 0.0078>, rot=<-0.0002 0.0098 0.1445 -0.9895>}
 ring_metacarpal_r {parent=hand_r, loc=<-2.2073 0.3627 -1.3638>, rot=<-0.0284 0.0722 -0.0458 -0.9959>}
 ring_01_r {parent=ring_metacarpal_r, loc=<-7.0297 0.6197 0.1446>, rot=<-0.0489 -0.0252 0.1892 -0.9804>}
 ring_02_r {parent=ring_01_r, loc=<-3.6171 -0.4648 0.1027>, rot=<0 0 0.2806 -0.9598>}
 ring_03_r {parent=ring_02_r, loc=<-2.4153 -0.1141 -0.0199>, rot=<0 0 0.1189 -0.9929>}
 thumb_01_r {parent=hand_r, loc=<-1.3119 0.3192 2.459>, rot=<0.6546 -0.4042 0.0664 -0.6354>}
 thumb_02_r {parent=thumb_01_r, loc=<-5.9336 0.0447 -0.0401>, rot=<-0.0037 0.0213 0.298 -0.9543>}
 thumb_03_r {parent=thumb_02_r, loc=<-2.2223 -0.3682 0.0907>, rot=<0.0118 -0.013 0.0066 -0.9998>}
 weapon_r {parent=hand_r, loc=<0.1975 3.4112 -0.3811>, rot=<-0.1365 -0.0242 0.9887 -0.0564>}
 lowerarm_twist_01_r {parent=lowerarm_r, loc=<-17.855 0 0>, rot=<0 0 0 1>}
 lowerarm_twist_02_r {parent=lowerarm_r, loc=<-9.6199 0 0>, rot=<0 0 0 1>}
 upperarm_twist_01_r {parent=upperarm_r, loc=<-0.0041 0 0>, rot=<0 0 0 1>}
 upperarm_twist_02_r {parent=upperarm_r, loc=<-13.7942 0 0>, rot=<0 0 0 1>}
 dyn_r_shoulderPad {parent=upperarm_r, loc=<0.0041 0 0>, rot=<0 0 0 1>}
 dyn_r_clavicle {parent=clavicle_r, loc=<0 0 0>, rot=<0 0 0 1>}
 neck_01 {parent=spine_05, loc=<11.3138 0.1529 0>, rot=<0 0 -0.2003 -0.9797>}
 neck_02 {parent=neck_01, loc=<4.5627 0 0>, rot=<0 0 0.0504 -0.9987>}
 head {parent=neck_02, loc=<5.3914 0 0>, rot=<0 0 0.115 -0.9934>}
 dyn_l_collar_MSTR {parent=neck_02, loc=<-1.4329 -3.2831 -9.5576>, rot=<0.0312 -0.0036 0.1149 -0.9929>}
 dyn_l_collar {parent=dyn_l_collar_MSTR, loc=<0 0 0>, rot=<0 0 0 1>}
 dyn_collar_MSTR {parent=neck_02, loc=<-2.3038 -9.8205 0>, rot=<0 0 0.115 -0.9934>}
 dyn_collar {parent=dyn_collar_MSTR, loc=<0 0 0>, rot=<0 0 0 -1>}
 dyn_r_collar_MSTR {parent=neck_02, loc=<-1.4329 -3.2831 9.5575>, rot=<-0.1963 0.0227 0.1127 -0.9738>}
 dyn_r_collar {parent=dyn_r_collar_MSTR, loc=<0 0 0>, rot=<0 0 0 -1>}
 dyn_zipper_MSTR {parent=spine_05, loc=<-2.0114 13.7107 -4.124>, rot=<0.1075 -0.6989 0.0512 -0.7053>}
 dyn_zipper {parent=dyn_zipper_MSTR, loc=<0 0 0>, rot=<0 0 0 1>}
 dyn_bullet_a_MSTR {parent=spine_05, loc=<-5.3012 13.3362 7.8405>, rot=<0.2226 0.5426 0.774 -0.2387>}
 dyn_bullet_a_01 {parent=dyn_bullet_a_MSTR, loc=<0 0 0>, rot=<0 0 0 -1>}
 dyn_bullet_b_MSTR {parent=spine_05, loc=<-8.365 14.0212 4.8827>, rot=<0.2324 0.5815 0.7314 -0.27>}
 dyn_bullet_b_01 {parent=dyn_bullet_b_MSTR, loc=<0 0 0>, rot=<0 0 0 -1>}
 dyn_bullet_c_MSTR {parent=spine_05, loc=<-11.2898 14.5872 1.7299>, rot=<0.2136 0.6377 0.6849 -0.2803>}
 dyn_bullet_c_01 {parent=dyn_bullet_c_MSTR, loc=<0 0 0>, rot=<0 0 0 1>}
 dyn_l_chestStrap_MSTR {parent=spine_05, loc=<-5.3927 9.2829 -14.7204>, rot=<-0.0569 -0.8331 -0.5495 -0.0287>}
 dyn_l_chestStrap_01 {parent=dyn_l_chestStrap_MSTR, loc=<0 0 0>, rot=<0 0 0 1>}
 dyn_r_chestStrap_MSTR {parent=spine_05, loc=<-4.8473 9.3301 14.6279>, rot=<0.0398 0.5641 0.8244 0.0239>}
 dyn_r_chestStrap_01 {parent=dyn_r_chestStrap_MSTR, loc=<0 0 0>, rot=<0 0 0 1>}
 thigh_l {parent=pelvis, loc=<-2.5693 -0.0109 -7.7501>, rot=<0.0012 0.0572 0.0155 -0.9982>}
 calf_l {parent=thigh_l, loc=<-39.0484 0 0>, rot=<0.0006 -0.0161 -0.0391 -0.9991>}
 calf_twist_01_l {parent=calf_l, loc=<-35.8131 -0.0967 0.0397>, rot=<0.0237 0.0095 -0.0035 -0.9997>}
 calf_twist_02_l {parent=calf_l, loc=<-18.8064 -0.0967 0.0397>, rot=<0.0237 0.0095 -0.0035 -0.9997>}
 foot_l {parent=calf_l, loc=<-42.561 0.1077 0.4292>, rot=<0.0418 -0.0455 0.0343 -0.9975>}
 ball_l {parent=foot_l, loc=<-6.1599 13.8264 0.6806>, rot=<0.0001 -0.0001 -0.7063 -0.7079>}
 dyn_l_knee {parent=calf_l, loc=<0 0 0>, rot=<0 0 0 1>}
 dyn_l_kneeStrapFix {parent=calf_l, loc=<3.9246 0 0>, rot=<0 0 0 1>}
 thigh_twist_01_l {parent=thigh_l, loc=<-21.6303 0 0>, rot=<0.0013 0 0 -1>}
 thigh_r {parent=pelvis, loc=<-2.5691 -0.0109 7.7501>, rot=<0.0572 -0.0012 0.9982 0.0155>}
 calf_r {parent=thigh_r, loc=<39.0487 0 0>, rot=<0.0006 -0.0161 -0.0391 -0.9991>}
 calf_twist_01_r {parent=calf_r, loc=<35.8132 0.0967 -0.0397>, rot=<0.0237 0.0095 -0.0035 -0.9997>}
 calf_twist_02_r {parent=calf_r, loc=<18.8064 0.0967 -0.0397>, rot=<0.0237 0.0095 -0.0035 -0.9997>}
 foot_r {parent=calf_r, loc=<42.5612 -0.1077 -0.4292>, rot=<0.0418 -0.0455 0.0343 -0.9975>}
 ball_r {parent=foot_r, loc=<6.1598 -13.8264 -0.6806>, rot=<0.0001 -0.0001 -0.7063 -0.7079>}
 dyn_r_knee {parent=calf_r, loc=<0 0 0>, rot=<0 0 0 1>}
 dyn_r_kneeClothFix {parent=calf_r, loc=<0 0 0>, rot=<0 0 0 1>}
 dyn_r_kneeStrapFix {parent=calf_r, loc=<-3.9192 0 0>, rot=<0 0 0 1>}
 thigh_twist_01_r {parent=thigh_r, loc=<21.6303 0 0>, rot=<0.0013 0 0 -1>}
 dyn_sword_MSTR {parent=pelvis, loc=<4.0168 -3.2514 -14.5615>, rot=<-0.0095 -0.9821 0.0466 -0.1823>}
 dyn_sword_01 {parent=dyn_sword_MSTR, loc=<0 0 0>, rot=<0 0 0 -1>}
 dyn_sword_02 {parent=dyn_sword_01, loc=<7.8545 0 0>, rot=<0.0433 -0.0694 -0.0134 -0.9966>}
 dyn_sword_03 {parent=dyn_sword_02, loc=<5.7484 0 0>, rot=<-0.0656 0.0101 0.016 -0.9977>}
 dyn_loinClothFront_l_MSTR {parent=pelvis, loc=<0.1964 9.6593 -9.915>, rot=<-0.0104 0.1116 0.9855 0.1274>}
 dyn_loinClothFront_01_l {parent=dyn_loinClothFront_l_MSTR, loc=<0 0 0>, rot=<0 0 0 1>}
 dyn_loinClothFront_02_l {parent=dyn_loinClothFront_01_l, loc=<9.5025 0 0>, rot=<-0.0425 -0.0109 -0.0698 -0.9966>}
 dyn_loinClothFront_03_l {parent=dyn_loinClothFront_02_l, loc=<6.3252 0 0>, rot=<-0.0271 -0.0184 -0.0213 -0.9992>}
 dyn_loinClothFront_c_MSTR {parent=pelvis, loc=<-0.4519 11.3492 0.0151>, rot=<-0.0034 0.0103 0.994 0.1086>}
 dyn_loinClothFront_01_c {parent=dyn_loinClothFront_c_MSTR, loc=<0 0 0>, rot=<0 0 0 1>}
 dyn_loinClothFront_02_c {parent=dyn_loinClothFront_01_c, loc=<9.2614 0 0>, rot=<0.0093 0.0004 -0.0328 -0.9994>}
 dyn_loinClothFront_03_c {parent=dyn_loinClothFront_02_c, loc=<6.5135 0 0>, rot=<-0.0022 -0.0003 -0.0374 -0.9993>}
 dyn_loinClothFront_r_MSTR {parent=pelvis, loc=<0.1713 9.5716 9.6348>, rot=<0.0195 -0.0279 0.9903 0.1345>}
 dyn_loinClothFront_01_r {parent=dyn_loinClothFront_r_MSTR, loc=<0 0 0>, rot=<0 0 0 1>}
 dyn_loinClothFront_02_r {parent=dyn_loinClothFront_01_r, loc=<9.5517 0 0>, rot=<-0.0141 -0.0043 -0.0725 -0.9973>}
 dyn_loinClothFront_03_r {parent=dyn_loinClothFront_02_r, loc=<6.5193 0 0>, rot=<0.0091 0.0019 -0.0127 -0.9999>}
 dyn_loinClothBack_r_MSTR {parent=pelvis, loc=<1.3218 -10.4605 9.5888>, rot=<0.0438 -0.9959 0.0786 0.0047>}
 dyn_loinClothBack_01_r {parent=dyn_loinClothBack_r_MSTR, loc=<0 0 0>, rot=<0 0 0 1>}
 dyn_loinClothBack_02_r {parent=dyn_loinClothBack_01_r, loc=<9.5025 0 0>, rot=<-0.0425 -0.0109 -0.0698 -0.9966>}
 dyn_loinClothBack_03_r {parent=dyn_loinClothBack_02_r, loc=<6.3252 0 0>, rot=<-0.0271 -0.0184 -0.0213 -0.9992>}
 dyn_loinClothBack_c_MSTR {parent=pelvis, loc=<1.3219 -11.3131 -0.4535>, rot=<0.0253 -0.9994 -0.0222 -0.0101>}
 dyn_loinClothBack_01_c {parent=dyn_loinClothBack_c_MSTR, loc=<0 0 0>, rot=<0 0 0 -1>}
 dyn_loinClothBack_02_c {parent=dyn_loinClothBack_01_c, loc=<9.2614 0 0>, rot=<0.0093 0.0004 -0.0328 -0.9994>}
 dyn_loinClothBack_03_c {parent=dyn_loinClothBack_02_c, loc=<6.5135 0 0>, rot=<-0.0022 -0.0003 -0.0374 -0.9993>}
 dyn_loinClothBack_l_MSTR {parent=pelvis, loc=<1.0979 -9.1797 -10.0291>, rot=<0.0511 -0.9963 -0.0579 -0.037>}
 dyn_loinClothBack_01_l {parent=dyn_loinClothBack_l_MSTR, loc=<0 0 0>, rot=<0 0 0 -1>}
 dyn_loinClothBack_02_l {parent=dyn_loinClothBack_01_l, loc=<9.5517 0 0>, rot=<-0.0141 -0.0043 -0.0725 -0.9973>}
 dyn_loinClothBack_03_l {parent=dyn_loinClothBack_02_l, loc=<6.5193 0 0>, rot=<0.0091 0.0019 -0.0127 -0.9999>}
 dyn_beltBuckle_MSTR {parent=pelvis, loc=<4.2138 12.1269 0>, rot=<0 0 0.022 -0.9998>}
 dyn_beltBuckle {parent=dyn_beltBuckle_MSTR, loc=<0 0 0>, rot=<0 0 0 -1>}
 ik_foot_root {parent=root, loc=<0 0 0>, rot=<0 0 0 1>}
 ik_foot_l {parent=ik_foot_root, loc=<15.2608 -4.4176 9.4432>, rot=<-0.0352 -0.6752 0.0068 0.7368>}
 ik_foot_r {parent=ik_foot_root, loc=<-15.2608 -4.4176 9.4431>, rot=<-0.7631 -0.0641 0.6388 0.0741>}
 ik_hand_root {parent=root, loc=<0 0 0>, rot=<0 0 0 1>}
 ik_hand_gun {parent=ik_hand_root, loc=<-51.9729 4.0884 102.6935>, rot=<-0.6084 -0.1793 -0.4613 0.6205>}
 ik_hand_l {parent=ik_hand_gun, loc=<53.0286 82.1782 35.2051>, rot=<-0.5102 -0.7906 -0.3387 0>}
 ik_hand_r {parent=ik_hand_gun, loc=<0 0 0>, rot=<0 0 0 -1>}
 dyn_root_simspace {parent=root, loc=<0 0 0>, rot=<0 0 0 1>}
 dyn_root_slope {parent=root, loc=<0 0 0>, rot=<0 0 0 1>}
]
