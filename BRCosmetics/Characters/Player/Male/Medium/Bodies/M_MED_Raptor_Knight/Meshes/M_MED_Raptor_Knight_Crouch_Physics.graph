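# Physics asset for the crouch setup of the male medium Raptor Knight cosmetic body, covering its loincloth and sword secondary motion.
# Six loincloth strips and the sword are three-box chains with limited angular constraints, tightest on the front-left strip.
# Chains hang from tiny capsule master bodies, which, like those for collar, straps, bullets and zipper, constrain to the pelvis with free rotation.
# Pelvis and thigh capsules supply the body shapes.
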PhysicsAsset /BRCosmetics/Characters/Player/Male/Medium/Bodies/M_MED_Raptor_Knight/Meshes/M_MED_Raptor_Knight_Crouch_Physics
bodies {
 dyn_bullet_a_MSTR [{shapeType=Capsule, length=0.505, radius=0.505}]
 dyn_bullet_b_MSTR [{shapeType=Capsule, length=0.505, radius=0.505}]
 dyn_bullet_c_MSTR [{shapeType=Capsule, length=0.505, radius=0.505}]
 dyn_collar_MSTR [{shapeType=Capsule, length=0.505, radius=0.505}]
 dyn_l_chestStrap_MSTR [{shapeType=Capsule, length=0.505, radius=0.505}]
 dyn_l_collar_MSTR [{shapeType=Capsule, length=0.505, radius=0.505}]
 dyn_loinClothBack_01_c [{shapeType=Box, extentX=8.7166, extentY=3.4013, extentZ=12.9719, center=<4.3592 -0.0001 0>}]
 dyn_loinClothBack_01_l [{shapeType=Box, extentX=8.7166, extentY=2.1183, extentZ=10.5765, center=<4.2866 0 0>}]
 dyn_loinClothBack_01_r [{shapeType=Box, extentX=8.7166, extentY=3.9096, extentZ=12.9719, center=<4.2547 0.0002 0>}]
 dyn_loinClothBack_02_c [{shapeType=Box, extentX=8.7166, extentY=2.2843, extentZ=12.9719, center=<4.5431 0.0001 0>}]
 dyn_loinClothBack_02_l [{shapeType=Box, extentX=8.7166, extentY=4.4309, extentZ=9.66, center=<3.1714 0 0>}]
 dyn_loinClothBack_02_r [{shapeType=Box, extentX=8.7166, extentY=5.1713, extentZ=12.9719, center=<4.5377 0.0001 0>}]
 dyn_loinClothBack_03_c [{shapeType=Box, extentX=8.7166, extentY=3.6479, extentZ=12.9719, center=<4.0298 0 0>}]
 dyn_loinClothBack_03_l [{shapeType=Box, extentX=5.5269, extentY=3.6479, extentZ=9.2405, center=<2.4677 0 -0.1359>}]
 dyn_loinClothBack_03_r [{shapeType=Box, extentX=8.7166, extentY=3.6479, extentZ=12.9719, center=<4.3817 0 0>}]
 dyn_loinClothBack_c_MSTR [{shapeType=Capsule, length=0.505, radius=0.505}]
 dyn_loinClothBack_l_MSTR [{shapeType=Capsule, length=0.505, radius=0.505}]
 dyn_loinClothBack_r_MSTR [{shapeType=Capsule, length=0.505, radius=0.505}]
 dyn_loinClothFront_01_c [{shapeType=Box, extentX=10.3375, extentY=0.5008, extentZ=11.808, center=<5.0603 -0.0003 0>}]
 dyn_loinClothFront_01_l [{shapeType=Box, extentX=8.5578, extentY=3.364, extentZ=14.5667, center=<4.3529 -0.0001 0>}]
 dyn_loinClothFront_01_r [{shapeType=Box, extentX=5.9747, extentY=2.0099, extentZ=14.5667, center=<5.3658 0.0001 0>, rotation=<rot 0 6.0269 0>}]
 dyn_loinClothFront_02_c [{shapeType=Box, extentX=6.8721, extentY=3.6853, extentZ=16.6263, center=<3.3237 -0.0003 0>}]
 dyn_loinClothFront_02_l [{shapeType=Box, extentX=8.5578, extentY=2, extentZ=14.5667, center=<4.4223 0 0>}]
 dyn_loinClothFront_02_r [{shapeType=Box, extentX=8.5578, extentY=3.1178, extentZ=14.5667, center=<4.1032 0 0>}]
 dyn_loinClothFront_03_c [{shapeType=Box, extentX=8.5578, extentY=3.6853, extentZ=16.0361, center=<4.136 -0.0002 0>}]
 dyn_loinClothFront_03_l [{shapeType=Box, extentX=8.5578, extentY=2, extentZ=14.5667, center=<4.0367 0.0001 0>}]
 dyn_loinClothFront_03_r [{shapeType=Box, extentX=8.5578, extentY=4.8619, extentZ=14.5667, center=<4.0762 0 0>}]
 dyn_loinClothFront_c_MSTR [{shapeType=Capsule, length=0.505, radius=0.505}]
 dyn_loinClothFront_l_MSTR [{shapeType=Capsule, length=0.505, radius=0.505}]
 dyn_loinClothFront_r_MSTR [{shapeType=Capsule, length=0.505, radius=0.505}]
 dyn_r_chestStrap_MSTR [{shapeType=Capsule, length=0.505, radius=0.505}]
 dyn_r_collar_MSTR [{shapeType=Capsule, length=0.505, radius=0.505}]
 dyn_sword_01 [{shapeType=Box, extentX=7.6889, extentY=30.0924, extentZ=0.5, center=<3.7743 0 0>}]
 dyn_sword_02 [{shapeType=Box, extentX=7.6889, extentY=30.0924, extentZ=0.5, center=<3.7743 0 0>}]
 dyn_sword_03 [{shapeType=Box, extentX=10.882, extentY=83.6396, extentZ=1.0093, center=<3.2285 -0.0717 0.4122>, rotation=<rot -3.2067 35.3256 -1.9446>}]
 dyn_sword_MSTR [{shapeType=Capsule, length=0.505, radius=0.505}]
 dyn_zipper_MSTR [{shapeType=Capsule, length=0.505, radius=0.505}]
 pelvis [{shapeType=Capsule, length=15.3336, radius=9.3382}]
 thigh_l [{shapeType=Capsule, length=33.3309, radius=10.6105, center=<-18.9793 0.4695 0>, rotation=<rot 90 0 0>}]
 thigh_r [{shapeType=Capsule, length=21.5688, radius=10.7485, center=<18.1538 -2.2647 0>, rotation=<rot 90 15.1541 22.1036>}]
}
constraints [
 {
  bone1Name=dyn_l_chestStrap_MSTR
  bone2Name=pelvis
  swing1LimitDegrees=45
  swing1Motion=Free
  swing2LimitDegrees=45
  swing2Motion=Free
  twistLimitDegrees=45
  twistMotion=Free
 }
 {
  bone1Name=dyn_r_chestStrap_MSTR
  bone2Name=pelvis
  swing1LimitDegrees=45
  swing1Motion=Free
  swing2LimitDegrees=45
  swing2Motion=Free
  twistLimitDegrees=45
  twistMotion=Free
 }
 {
  bone1Name=dyn_bullet_a_MSTR
  bone2Name=pelvis
  swing1LimitDegrees=45
  swing1Motion=Free
  swing2LimitDegrees=45
  swing2Motion=Free
  twistLimitDegrees=45
  twistMotion=Free
 }
 {
  bone1Name=dyn_bullet_b_MSTR
  bone2Name=pelvis
  swing1LimitDegrees=45
  swing1Motion=Free
  swing2LimitDegrees=45
  swing2Motion=Free
  twistLimitDegrees=45
  twistMotion=Free
 }
 {
  bone1Name=dyn_bullet_c_MSTR
  bone2Name=pelvis
  swing1LimitDegrees=45
  swing1Motion=Free
  swing2LimitDegrees=45
  swing2Motion=Free
  twistLimitDegrees=45
  twistMotion=Free
 }
 {
  bone1Name=dyn_sword_MSTR
  bone2Name=pelvis
  swing1LimitDegrees=45
  swing1Motion=Free
  swing2LimitDegrees=45
  swing2Motion=Free
  twistLimitDegrees=45
  twistMotion=Free
 }
 {
  bone1Name=thigh_l
  bone2Name=pelvis
  swing1LimitDegrees=45
  swing1Motion=Free
  swing2LimitDegrees=45
  swing2Motion=Free
  twistLimitDegrees=45
  twistMotion=Free
 }
 {
  bone1Name=thigh_r
  bone2Name=pelvis
  swing1LimitDegrees=45
  swing1Motion=Free
  swing2LimitDegrees=45
  swing2Motion=Free
  twistLimitDegrees=45
  twistMotion=Free
 }
 {bone1Name=dyn_loinClothBack_01_c, bone2Name=dyn_loinClothBack_c_MSTR, swing1LimitDegrees=60, swing2LimitDegrees=30, twistLimitDegrees=30}
 {bone1Name=dyn_loinClothBack_02_c, bone2Name=dyn_loinClothBack_01_c, swing1LimitDegrees=60, swing2LimitDegrees=30, twistLimitDegrees=30}
 {bone1Name=dyn_loinClothBack_03_c, bone2Name=dyn_loinClothBack_02_c, swing1LimitDegrees=60, swing2LimitDegrees=30, twistLimitDegrees=30}
 {bone1Name=dyn_loinClothBack_01_r, bone2Name=dyn_loinClothBack_r_MSTR, swing1LimitDegrees=60, swing2LimitDegrees=30, twistLimitDegrees=30}
 {bone1Name=dyn_loinClothBack_02_r, bone2Name=dyn_loinClothBack_01_r, swing1LimitDegrees=60, swing2LimitDegrees=30, twistLimitDegrees=30}
 {bone1Name=dyn_loinClothBack_03_r, bone2Name=dyn_loinClothBack_02_r, swing1LimitDegrees=60, swing2LimitDegrees=30, twistLimitDegrees=30}
 {bone1Name=dyn_loinClothBack_01_l, bone2Name=dyn_loinClothBack_l_MSTR, swing1LimitDegrees=60, swing2LimitDegrees=10, twistLimitDegrees=15}
 {bone1Name=dyn_loinClothBack_02_l, bone2Name=dyn_loinClothBack_01_l, swing1LimitDegrees=60, swing2LimitDegrees=10, twistLimitDegrees=15}
 {bone1Name=dyn_loinClothBack_03_l, bone2Name=dyn_loinClothBack_02_l, swing1LimitDegrees=60, swing2LimitDegrees=10, twistLimitDegrees=15}
 {
  bone1Name=dyn_loinClothBack_r_MSTR
  bone2Name=pelvis
  swing1LimitDegrees=45
  swing1Motion=Free
  swing2LimitDegrees=45
  swing2Motion=Free
  twistLimitDegrees=45
  twistMotion=Free
 }
 {
  bone1Name=dyn_loinClothBack_c_MSTR
  bone2Name=pelvis
  swing1LimitDegrees=45
  swing1Motion=Free
  swing2LimitDegrees=45
  swing2Motion=Free
  twistLimitDegrees=45
  twistMotion=Free
 }
 {
  bone1Name=dyn_loinClothBack_l_MSTR
  bone2Name=pelvis
  swing1LimitDegrees=45
  swing1Motion=Free
  swing2LimitDegrees=45
  swing2Motion=Free
  twistLimitDegrees=45
  twistMotion=Free
 }
 {bone1Name=dyn_sword_01, bone2Name=dyn_sword_MSTR, swing1LimitDegrees=20, swing2LimitDegrees=20, twistLimitDegrees=20}
 {bone1Name=dyn_sword_02, bone2Name=dyn_sword_01, swing1LimitDegrees=20, swing2LimitDegrees=20, twistLimitDegrees=20}
 {bone1Name=dyn_sword_03, bone2Name=dyn_sword_02, swing1LimitDegrees=20, swing2LimitDegrees=20, twistLimitDegrees=20}
 {
  bone1Name=dyn_l_collar_MSTR
  bone2Name=pelvis
  swing1LimitDegrees=45
  swing1Motion=Free
  swing2LimitDegrees=45
  swing2Motion=Free
  twistLimitDegrees=45
  twistMotion=Free
 }
 {
  bone1Name=dyn_collar_MSTR
  bone2Name=pelvis
  swing1LimitDegrees=45
  swing1Motion=Free
  swing2LimitDegrees=45
  swing2Motion=Free
  twistLimitDegrees=45
  twistMotion=Free
 }
 {
  bone1Name=dyn_r_collar_MSTR
  bone2Name=pelvis
  swing1LimitDegrees=45
  swing1Motion=Free
  swing2LimitDegrees=45
  swing2Motion=Free
  twistLimitDegrees=45
  twistMotion=Free
 }
 {
  bone1Name=dyn_zipper_MSTR
  bone2Name=pelvis
  swing1LimitDegrees=45
  swing1Motion=Free
  swing2LimitDegrees=45
  swing2Motion=Free
  twistLimitDegrees=45
  twistMotion=Free
 }
 {
  bone1Name=dyn_loinClothFront_l_MSTR
  bone2Name=pelvis
  swing1LimitDegrees=45
  swing1Motion=Free
  swing2LimitDegrees=45
  swing2Motion=Free
  twistLimitDegrees=45
  twistMotion=Free
 }
 {
  bone1Name=dyn_loinClothFront_c_MSTR
  bone2Name=pelvis
  swing1LimitDegrees=45
  swing1Motion=Free
  swing2LimitDegrees=45
  swing2Motion=Free
  twistLimitDegrees=45
  twistMotion=Free
 }
 {
  bone1Name=dyn_loinClothFront_r_MSTR
  bone2Name=pelvis
  swing1LimitDegrees=45
  swing1Motion=Free
  swing2LimitDegrees=45
  swing2Motion=Free
  twistLimitDegrees=45
  twistMotion=Free
 }
 {bone1Name=dyn_loinClothFront_01_l, bone2Name=dyn_loinClothFront_l_MSTR, swing1LimitDegrees=5, swing2LimitDegrees=5, twistLimitDegrees=5}
 {bone1Name=dyn_loinClothFront_02_l, bone2Name=dyn_loinClothFront_01_l, swing1LimitDegrees=5, swing2LimitDegrees=5, twistLimitDegrees=5}
 {bone1Name=dyn_loinClothFront_03_l, bone2Name=dyn_loinClothFront_02_l, swing1LimitDegrees=5, swing2LimitDegrees=5, twistLimitDegrees=5}
 {bone1Name=dyn_loinClothFront_01_c, bone2Name=dyn_loinClothFront_c_MSTR, swing1LimitDegrees=45, swing2LimitDegrees=45, twistLimitDegrees=45}
 {bone1Name=dyn_loinClothFront_02_c, bone2Name=dyn_loinClothFront_01_c, swing1LimitDegrees=45, swing2LimitDegrees=45, twistLimitDegrees=45}
 {bone1Name=dyn_loinClothFront_03_c, bone2Name=dyn_loinClothFront_02_c, swing1LimitDegrees=45, swing2LimitDegrees=45, twistLimitDegrees=45}
 {bone1Name=dyn_loinClothFront_01_r, bone2Name=dyn_loinClothFront_r_MSTR, swing1LimitDegrees=45, swing2LimitDegrees=45, twistLimitDegrees=45}
 {bone1Name=dyn_loinClothFront_02_r, bone2Name=dyn_loinClothFront_01_r, swing1LimitDegrees=45, swing2LimitDegrees=45, twistLimitDegrees=45}
 {bone1Name=dyn_loinClothFront_03_r, bone2Name=dyn_loinClothFront_02_r, swing1LimitDegrees=45, swing2LimitDegrees=45, twistLimitDegrees=45}
]
