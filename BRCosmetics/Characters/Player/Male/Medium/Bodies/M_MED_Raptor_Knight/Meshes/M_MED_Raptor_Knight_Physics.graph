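# Physics asset for the male medium Raptor Knight body cosmetic, providing jointed bodies for its loincloth.
# Front and back cloths each form left, center and right chains of three boxes with angular limits, with much tighter twist on front-left.
# Chain roots, thigh capsules and tiny bullet, collar, chest-strap and zipper master capsules all attach to the pelvis with free rotation.
# Pelvis and thigh capsules are the only body-sized shapes.
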PhysicsAsset /BRCosmetics/Characters/Player/Male/Medium/Bodies/M_MED_Raptor_Knight/Meshes/M_MED_Raptor_Knight_Physics
bodies {
 dyn_bullet_a_MSTR [{shapeType=Capsule, length=0.505, radius=0.505}]
 dyn_bullet_b_MSTR [{shapeType=Capsule, length=0.505, radius=0.505}]
 dyn_bullet_c_MSTR [{shapeType=Capsule, length=0.505, radius=0.505}]
 dyn_collar_MSTR [{shapeType=Capsule, length=0.505, radius=0.505}]
 dyn_l_chestStrap_MSTR [{shapeType=Capsule, length=0.505, radius=0.505}]
 dyn_l_collar_MSTR [{shapeType=Capsule, length=0.505, radius=0.505}]
 dyn_loinClothBack_01_c [{shapeType=Box, extentX=8.7166, extentY=3.4013, extentZ=12.9719, center=<4.3592 -0.0001 0>}]
 dyn_loinClothBack_01_l [{shapeType=Box, extentX=8.7166, extentY=1.0339, extentZ=8.145, center=<4.2866 0 0>}]
 dyn_loinClothBack_01_r [{shapeType=Box, extentX=8.7166, extentY=3.9096, extentZ=12.9719, center=<4.2547 0.0002 0>}]
 dyn_loinClothBack_02_c [{shapeType=Box, extentX=8.7166, extentY=2.2843, extentZ=12.9719, center=<4.5431 0.0001 0>}]
 dyn_loinClothBack_02_l [{shapeType=Box, extentX=8.7166, extentY=2.2201, extentZ=8.3207, center=<3.1714 0 0>}]
 dyn_loinClothBack_02_r [{shapeType=Box, extentX=8.7166, extentY=5.1713, extentZ=12.9719, center=<4.5377 0.0001 0>}]
 dyn_loinClothBack_03_c [{shapeType=Box, extentX=8.7166, extentY=3.6479, extentZ=12.9719, center=<4.0298 0 0>}]
 dyn_loinClothBack_03_l [{shapeType=Box, extentX=8.7166, extentY=3.6479, extentZ=9.5055, center=<4.0994 0 0>}]
 dyn_loinClothBack_03_r [{shapeType=Box, extentX=8.7166, extentY=3.6479, extentZ=12.9719, center=<4.3817 0 0>}]
 dyn_loinClothBack_c_MSTR [{shapeType=Capsule, length=0.505, radius=0.505}]
 dyn_loinClothBack_l_MSTR [{shapeType=Capsule, length=0.505, radius=0.505}]
 dyn_loinClothBack_r_MSTR [{shapeType=Capsule, length=0.505, radius=0.505}]
 dyn_loinClothFront_01_c [{shapeType=Box, extentX=8.5578, extentY=3.364, extentZ=14.5667, center=<4.2063 0.0001 0>}]
 dyn_loinClothFront_01_l [{shapeType=Box, extentX=8.5578, extentY=2.3199, extentZ=7.9138, center=<4.3529 -0.0001 0>}]
 dyn_loinClothFront_01_r [{shapeType=Box, extentX=8.5578, extentY=2.327, extentZ=14.5667, center=<4.2557 0.0001 0>}]
 dyn_loinClothFront_02_c [{shapeType=Box, extentX=8.5578, extentY=2.0891, extentZ=14.5667, center=<4.1271 0 0>}]
 dyn_loinClothFront_02_l [{shapeType=Box, extentX=8.5578, extentY=3.6047, extentZ=7.9138, center=<4.4223 0 0>}]
 dyn_loinClothFront_02_r [{shapeType=Box, extentX=8.5578, extentY=2.1108, extentZ=14.5667, center=<4.1032 0 0>}]
 dyn_loinClothFront_03_c [{shapeType=Box, extentX=8.5578, extentY=4.0399, extentZ=14.5667, center=<4.0953 0 0>}]
 dyn_loinClothFront_03_l [{shapeType=Box, extentX=8.5578, extentY=3.6047, extentZ=7.9138, center=<4.0367 0.0001 0>}]
 dyn_loinClothFront_03_r [{shapeType=Box, extentX=8.5578, extentY=2.1108, extentZ=14.5667, center=<4.0762 0 0>}]
 dyn_loinClothFront_c_MSTR [{shapeType=Capsule, length=0.505, radius=0.505}]
 dyn_loinClothFront_l_MSTR [{shapeType=Capsule, length=0.505, radius=0.505}]
 dyn_loinClothFront_r_MSTR [{shapeType=Capsule, length=0.505, radius=0.505}]
 dyn_r_chestStrap_MSTR [{shapeType=Capsule, length=0.505, radius=0.505}]
 dyn_r_collar_MSTR [{shapeType=Capsule, length=0.505, radius=0.505}]
 dyn_zipper_MSTR [{shapeType=Capsule, length=0.505, radius=0.505}]
 pelvis [{shapeType=Capsule, length=15.3336, radius=9.3382}]
 thigh_l [{shapeType=Capsule, length=51.4591, radius=10.3573, center=<-18.9793 0.3089 0>, rotation=<rot 90 0 0>}]
 thigh_r [{shapeType=Capsule, length=46.9236, radius=10.9081, center=<18.4298 -0.332 0>, rotation=<rot 90 0 0>}]
}
constraints [
 {
  bone1Name=dyn_l_chestStrap_MSTR
  bone2Name=pelvis
  swing1LimitDegrees=45
  swing1Motion=Free
  swing2LimitDegrees=45
  swing2Motion=Free
  twistLimitDegrees=45
  twistMotion=Free
 }
 {
  bone1Name=dyn_r_chestStrap_MSTR
  bone2Name=pelvis
  swing1LimitDegrees=45
  swing1Motion=Free
  swing2LimitDegrees=45
  swing2Motion=Free
  twistLimitDegrees=45
  twistMotion=Free
 }
 {
  bone1Name=dyn_bullet_a_MSTR
  bone2Name=pelvis
  swing1LimitDegrees=45
  swing1Motion=Free
  swing2LimitDegrees=45
  swing2Motion=Free
  twistLimitDegrees=45
  twistMotion=Free
 }
 {
  bone1Name=dyn_bullet_b_MSTR
  bone2Name=pelvis
  swing1LimitDegrees=45
  swing1Motion=Free
  swing2LimitDegrees=45
  swing2Motion=Free
  twistLimitDegrees=45
  twistMotion=Free
 }
 {
  bone1Name=dyn_bullet_c_MSTR
  bone2Name=pelvis
  swing1LimitDegrees=45
  swing1Motion=Free
  swing2LimitDegrees=45
  swing2Motion=Free
  twistLimitDegrees=45
  twistMotion=Free
 }
 {
  bone1Name=thigh_l
  bone2Name=pelvis
  swing1LimitDegrees=45
  swing1Motion=Free
  swing2LimitDegrees=45
  swing2Motion=Free
  twistLimitDegrees=45
  twistMotion=Free
 }
 {
  bone1Name=thigh_r
  bone2Name=pelvis
  swing1LimitDegrees=45
  swing1Motion=Free
  swing2LimitDegrees=45
  swing2Motion=Free
  twistLimitDegrees=45
  twistMotion=Free
 }
 {bone1Name=dyn_loinClothBack_01_c, bone2Name=dyn_loinClothBack_c_MSTR, swing1LimitDegrees=60, swing2LimitDegrees=30, twistLimitDegrees=30}
 {bone1Name=dyn_loinClothBack_02_c, bone2Name=dyn_loinClothBack_01_c, swing1LimitDegrees=60, swing2LimitDegrees=30, twistLimitDegrees=30}
 {bone1Name=dyn_loinClothBack_03_c, bone2Name=dyn_loinClothBack_02_c, swing1LimitDegrees=60, swing2LimitDegrees=30, twistLimitDegrees=30}
 {bone1Name=dyn_loinClothBack_01_r, bone2Name=dyn_loinClothBack_r_MSTR, swing1LimitDegrees=60, swing2LimitDegrees=30, twistLimitDegrees=30}
 {bone1Name=dyn_loinClothBack_02_r, bone2Name=dyn_loinClothBack_01_r, swing1LimitDegrees=60, swing2LimitDegrees=30, twistLimitDegrees=30}
 {bone1Name=dyn_loinClothBack_03_r, bone2Name=dyn_loinClothBack_02_r, swing1LimitDegrees=60, swing2LimitDegrees=30, twistLimitDegrees=30}
 {bone1Name=dyn_loinClothBack_01_l, bone2Name=dyn_loinClothBack_l_MSTR, swing1LimitDegrees=60, swing2LimitDegrees=30, twistLimitDegrees=30}
 {bone1Name=dyn_loinClothBack_02_l, bone2Name=dyn_loinClothBack_01_l, swing1LimitDegrees=60, swing2LimitDegrees=30, twistLimitDegrees=30}
 {bone1Name=dyn_loinClothBack_03_l, bone2Name=dyn_loinClothBack_02_l, swing1LimitDegrees=60, swing2LimitDegrees=30, twistLimitDegrees=30}
 {
  bone1Name=dyn_loinClothBack_r_MSTR
  bone2Name=pelvis
  swing1LimitDegrees=45
  swing1Motion=Free
  swing2LimitDegrees=45
  swing2Motion=Free
  twistLimitDegrees=45
  twistMotion=Free
 }
 {
  bone1Name=dyn_loinClothBack_c_MSTR
  bone2Name=pelvis
  swing1LimitDegrees=45
  swing1Motion=Free
  swing2LimitDegrees=45
  swing2Motion=Free
  twistLimitDegrees=45
  twistMotion=Free
 }
 {
  bone1Name=dyn_loinClothBack_l_MSTR
  bone2Name=pelvis
  swing1LimitDegrees=45
  swing1Motion=Free
  swing2LimitDegrees=45
  swing2Motion=Free
  twistLimitDegrees=45
  twistMotion=Free
 }
 {
  bone1Name=dyn_l_collar_MSTR
  bone2Name=pelvis
  swing1LimitDegrees=45
  swing1Motion=Free
  swing2LimitDegrees=45
  swing2Motion=Free
  twistLimitDegrees=45
  twistMotion=Free
 }
 {
  bone1Name=dyn_collar_MSTR
  bone2Name=pelvis
  swing1LimitDegrees=45
  swing1Motion=Free
  swing2LimitDegrees=45
  swing2Motion=Free
  twistLimitDegrees=45
  twistMotion=Free
 }
 {
  bone1Name=dyn_r_collar_MSTR
  bone2Name=pelvis
  swing1LimitDegrees=45
  swing1Motion=Free
  swing2LimitDegrees=45
  swing2Motion=Free
  twistLimitDegrees=45
  twistMotion=Free
 }
 {
  bone1Name=dyn_zipper_MSTR
  bone2Name=pelvis
  swing1LimitDegrees=45
  swing1Motion=Free
  swing2LimitDegrees=45
  swing2Motion=Free
  twistLimitDegrees=45
  twistMotion=Free
 }
 {
  bone1Name=dyn_loinClothFront_l_MSTR
  bone2Name=pelvis
  swing1LimitDegrees=45
  swing1Motion=Free
  swing2LimitDegrees=45
  swing2Motion=Free
  twistLimitDegrees=45
  twistMotion=Free
 }
 {
  bone1Name=dyn_loinClothFront_c_MSTR
  bone2Name=pelvis
  swing1LimitDegrees=45
  swing1Motion=Free
  swing2LimitDegrees=45
  swing2Motion=Free
  twistLimitDegrees=45
  twistMotion=Free
 }
 {
  bone1Name=dyn_loinClothFront_r_MSTR
  bone2Name=pelvis
  swing1LimitDegrees=45
  swing1Motion=Free
  swing2LimitDegrees=45
  swing2Motion=Free
  twistLimitDegrees=45
  twistMotion=Free
 }
 {bone1Name=dyn_loinClothFront_01_l, bone2Name=dyn_loinClothFront_l_MSTR, swing1LimitDegrees=45, swing2LimitDegrees=45, twistLimitDegrees=5}
 {bone1Name=dyn_loinClothFront_02_l, bone2Name=dyn_loinClothFront_01_l, swing1LimitDegrees=45, swing2LimitDegrees=45, twistLimitDegrees=5}
 {bone1Name=dyn_loinClothFront_03_l, bone2Name=dyn_loinClothFront_02_l, swing1LimitDegrees=45, swing2LimitDegrees=45, twistLimitDegrees=5}
 {bone1Name=dyn_loinClothFront_01_c, bone2Name=dyn_loinClothFront_c_MSTR, swing1LimitDegrees=45, swing2LimitDegrees=45, twistLimitDegrees=45}
 {bone1Name=dyn_loinClothFront_02_c, bone2Name=dyn_loinClothFront_01_c, swing1LimitDegrees=45, swing2LimitDegrees=45, twistLimitDegrees=45}
 {bone1Name=dyn_loinClothFront_03_c, bone2Name=dyn_loinClothFront_02_c, swing1LimitDegrees=45, swing2LimitDegrees=45, twistLimitDegrees=45}
 {bone1Name=dyn_loinClothFront_01_r, bone2Name=dyn_loinClothFront_r_MSTR, swing1LimitDegrees=45, swing2LimitDegrees=45, twistLimitDegrees=45}
 {bone1Name=dyn_loinClothFront_02_r, bone2Name=dyn_loinClothFront_01_r, swing1LimitDegrees=45, swing2LimitDegrees=45, twistLimitDegrees=45}
 {bone1Name=dyn_loinClothFront_03_r, bone2Name=dyn_loinClothFront_02_r, swing1LimitDegrees=45, swing2LimitDegrees=45, twistLimitDegrees=45}
]
