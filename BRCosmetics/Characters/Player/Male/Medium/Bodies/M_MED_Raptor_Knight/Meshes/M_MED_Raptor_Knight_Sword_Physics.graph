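# Physics asset for the Raptor Knight male medium body's sword, set up as a dynamic bone chain on the lower body.
# Three thin box bodies chain from a tiny dyn_sword_MSTR anchor capsule through angle-limited joints, with the last box larger and rotated.
# Capsules on the pelvis and both thighs give the body shapes around the sword.
# The anchor and both thighs attach to the pelvis with free swing and twist.
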PhysicsAsset /BRCosmetics/Characters/Player/Male/Medium/Bodies/M_MED_Raptor_Knight/Meshes/M_MED_Raptor_Knight_Sword_Physics
bodies {
 dyn_sword_01 [{shapeType=Box, extentX=7.6889, extentY=30.0924, extentZ=0.5, center=<3.7743 0 0>}]
 dyn_sword_02 [{shapeType=Box, extentX=7.6889, extentY=30.0924, extentZ=0.5, center=<3.7743 0 0>}]
 dyn_sword_03 [{shapeType=Box, extentX=10.882, extentY=83.6396, extentZ=1.0093, center=<3.2285 -0.0717 0.4122>, rotation=<rot -3.2067 35.3256 -1.9446>}]
 dyn_sword_MSTR [{shapeType=Capsule, length=0.505, radius=0.505}]
 pelvis [
  {shapeType=Capsule, length=15.3336, radius=9.3382}
  {shapeType=Capsule, length=109.1793, radius=9.3382, center=<-20.4983 -25.633 1.4834>, rotation=<rot 89.5098 179.9982 179.9982>}
 ]
 thigh_l [{shapeType=Capsule, length=35.1379, radius=8.8035, center=<-18.9793 0.0002 0>, rotation=<rot 90 0 0>}]
 thigh_r [{shapeType=Capsule, length=22.0889, radius=10.2283, center=<18.4298 0 0>, rotation=<rot 90 0 0>}]
}
constraints [
 {
  bone1Name=dyn_sword_MSTR
  bone2Name=pelvis
  swing1LimitDegrees=45
  swing1Motion=Free
  swing2LimitDegrees=45
  swing2Motion=Free
  twistLimitDegrees=45
  twistMotion=Free
 }
 {
  bone1Name=thigh_l
  bone2Name=pelvis
  swing1LimitDegrees=45
  swing1Motion=Free
  swing2LimitDegrees=45
  swing2Motion=Free
  twistLimitDegrees=45
  twistMotion=Free
 }
 {
  bone1Name=thigh_r
  bone2Name=pelvis
  swing1LimitDegrees=45
  swing1Motion=Free
  swing2LimitDegrees=45
  swing2Motion=Free
  twistLimitDegrees=45
  twistMotion=Free
 }
 {bone1Name=dyn_sword_01, bone2Name=dyn_sword_MSTR, swing1LimitDegrees=50, swing2LimitDegrees=45, twistLimitDegrees=20}
 {bone1Name=dyn_sword_02, bone2Name=dyn_sword_01, swing1LimitDegrees=50, swing2LimitDegrees=45, twistLimitDegrees=20}
 {bone1Name=dyn_sword_03, bone2Name=dyn_sword_02, swing1LimitDegrees=50, swing2LimitDegrees=45, twistLimitDegrees=20}
]
