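# Anim Blueprint for the Raptor Knight body cosmetic (LowMS) that drives secondary motion on its sword, loincloth, pads, straps, bullets and zipper.
# Cached 'ConstrainPass' pose feeds emote, crouch and default branches, each mixing RigidBody sims, Control Rigs and bone constraints, with corrective offsets in some.
# Blended result gets skydive and aim adjustments, then weighted multi-bone constraints anchor accessory root bones ahead of chained AnimDynamics springs.
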AnimBlueprint /BRCosmetics/Characters/Player/Male/Medium/Bodies/M_MED_Raptor_Knight/Meshes/M_MED_Raptor_Knight_LowMS_AnimBP
target_skeleton=&/M_MED_Raptor_Knight_Skeleton
graphs [
 AnimGraph {nodes [
   n1 = Control Rig(Source=Use cached pose 'ConstrainPass'(), Alpha=Get ArmLiftRAlpha() -> (ArmLiftRAlpha:float)) {
    alphaBoolBlend {blendOption=Linear}
    controlRigAssetReference {blueprintRigClass=&/M_MED_Raptor_Knight_Pads_CR.M_MED_Raptor_Knight_Pads_CR_C}
   }
   n2 = Use cached pose 'ConstrainPass'()
   n3 = Get bIsSkydiving_Any() -> (bIsSkydiving_Any:bool)
   n4 = Get bIsJumpingOrFalling() -> (bIsJumpingOrFalling:bool)
   n5 = "Transform (Modify) Bone - Bone: dyn_l_knee"!(
    ComponentPose=Local To Component(LocalPose=Blend Poses by bool(BlendPose_0=@n2, BlendPose_1=@n1, bActiveValue=@n3) {blendType=Linear})
    bAlphaBoolEnabled:bool=@n4
    Translation="0,0,0"
    Rotation="0,0,0"
    Scale="1,1,1"
   ) -> (Pose) {
    alphaInputType=Bool
    rotationMode=BMM_Additive
    rotationSpace=BCS_BoneSpace
    scaleMode=BMM_Additive
    scaleSpace=BCS_BoneSpace
    translationMode=BMM_Additive
    translationSpace=BCS_BoneSpace
    alphaBoolBlend {blendInTime=0.1, blendOption=Linear, blendOutTime=0.1}
    boneToModify {boneName=dyn_l_knee}
   }
   n6 = "Transform (Modify) Bone - Bone: dyn_r_knee"!(ComponentPose=@n5, bAlphaBoolEnabled:bool=@n4, Translation="1,3.5,0", Rotation="0.000,10,0.000", Scale="1,1,1") -> (Pose) {
    alphaInputType=Bool
    rotationMode=BMM_Additive
    rotationSpace=BCS_BoneSpace
    scaleMode=BMM_Additive
    scaleSpace=BCS_BoneSpace
    translationMode=BMM_Additive
    translationSpace=BCS_BoneSpace
    alphaBoolBlend {blendInTime=0.1, blendOption=Linear, blendOutTime=0.1}
    boneToModify {boneName=dyn_r_knee}
    rotation {yaw=10}
    translation=<1 3.5>
   }
   n7 = RigidBody!(
    ComponentPose=@n6
    Alpha=Get IsNotFrontEndAlpha() -> (IsNotFrontEndAlpha:float)
    OverrideWorldGravity:Vector=Get GravityOverride()
    ExternalForce
   ) -> (Pose) {
    bClampLinearTranslationLimitToRefPose=true
    bOverrideWorldGravity=true
    overridePhysicsAsset=&/M_MED_Raptor_Knight_Physics
    simulationSpace=BaseBoneSpace
    alphaBoolBlend {blendOption=Linear}
    baseBoneRef {boneName=pelvis}
    componentAppliedLinearAccClamp=<2000 2000 3000>
    componentLinearAccScale=<0.35 0.35 0.45>
    componentLinearVelScale=<0.1 0.1 0.25>
    overrideWorldGravity {z=-980}
   }
   n8 = Get IsFrontEndSwordConstraintsAlpha() -> (IsFrontEndSwordConstraintsAlpha:float)
   n9 = "Constraint - dyn_sword_MSTR (P)"(ComponentPose=@n7, Alpha=@n8, ConstraintWeights_0=1) {
    alphaBoolBlend {blendOption=Linear}
    boneToModify {boneName=dyn_sword_MSTR}
    constraintSetup [{targetBone {boneName=pelvis}}]
    constraintWeights [1]
   }
   n10 = "Constraint - dyn_sword_01 (P)"(ComponentPose=@n9, Alpha=@n8, ConstraintWeights_0=1) {
    alphaBoolBlend {blendOption=Linear}
    boneToModify {boneName=dyn_sword_01}
    constraintSetup [{targetBone {boneName=pelvis}}]
    constraintWeights [1]
   }
   n11 = "Constraint - dyn_sword_02 (P)"(ComponentPose=@n10, Alpha=@n8, ConstraintWeights_0=1) {
    alphaBoolBlend {blendOption=Linear}
    boneToModify {boneName=dyn_sword_02}
    constraintSetup [{targetBone {boneName=pelvis}}]
    constraintWeights [1]
   }
   n12 = "Constraint - dyn_sword_03 (P)"(ComponentPose=@n11, Alpha=@n8, ConstraintWeights_0=1) {
    alphaBoolBlend {blendOption=Linear}
    boneToModify {boneName=dyn_sword_03}
    constraintSetup [{targetBone {boneName=pelvis}}]
    constraintWeights [1]
   }
   n13 = "Transform (Modify) Bone - Bone: dyn_sword_01"!(ComponentPose=@n12, bAlphaBoolEnabled:bool=True, Translation="0,0,0", Rotation="0.000,0.000,-10", Scale="1,1,1") -> (Pose) {
    alphaInputType=Bool
    rotationMode=BMM_Additive
    rotationSpace=BCS_BoneSpace
    scaleMode=BMM_Additive
    scaleSpace=BCS_BoneSpace
    translationMode=BMM_Additive
    translationSpace=BCS_BoneSpace
    alphaBoolBlend {blendInTime=0.1, blendOption=Linear, blendOutTime=0.1}
    boneToModify {boneName=dyn_sword_01}
    rotation {roll=-10}
   }
   n14 = Get IsPlayingFullBodySlotInFrontEndAlpha() -> (IsPlayingFullBodySlotInFrontEndAlpha:float)
   n15 = "Constraint - dyn_loinClothFront_r_MSTR (P)"(ComponentPose=@n13, Alpha=@n14, ConstraintWeights_0=1) {
    alphaBoolBlend {blendOption=Linear}
    boneToModify {boneName=dyn_loinClothFront_r_MSTR}
    constraintSetup [{targetBone {boneName=thigh_r}}]
    constraintWeights [1]
   }
   n16 = "Constraint - dyn_loinClothFront_c_MSTR (PP)"!(ComponentPose=@n15, Alpha=@n14, ConstraintWeights_0=0.5, ConstraintWeights_1:float=0.5) -> (Pose) {
    alphaBoolBlend {blendOption=Linear}
    boneToModify {boneName=dyn_loinClothFront_c_MSTR}
    constraintSetup [{targetBone {boneName=thigh_r}}, {targetBone {boneName=thigh_l}}]
    constraintWeights [1,1]
   }
   n17 = "Constraint - dyn_loinClothFront_l_MSTR (P)"(ComponentPose=@n16, Alpha=@n14, ConstraintWeights_0=1) {
    alphaBoolBlend {blendOption=Linear}
    boneToModify {boneName=dyn_loinClothFront_l_MSTR}
    constraintSetup [{targetBone {boneName=thigh_l}}]
    constraintWeights [1]
   }
   n18 = "Constraint - dyn_loinClothFront_01_r (P)"(ComponentPose=@n17, Alpha=@n14, ConstraintWeights_0=1) {
    alphaBoolBlend {blendOption=Linear}
    boneToModify {boneName=dyn_loinClothFront_01_r}
    constraintSetup [{targetBone {boneName=thigh_r}}]
    constraintWeights [1]
   }
   n19 = "Constraint - dyn_loinClothFront_01_c (PP)"!(ComponentPose=@n18, Alpha=@n14, ConstraintWeights_0=0.5, ConstraintWeights_1:float=0.5) -> (Pose) {
    alphaBoolBlend {blendOption=Linear}
    boneToModify {boneName=dyn_loinClothFront_01_c}
    constraintSetup [{targetBone {boneName=thigh_r}}, {targetBone {boneName=thigh_l}}]
    constraintWeights [1,1]
   }
   n20 = "Constraint - dyn_loinClothFront_01_l (P)"(ComponentPose=@n19, Alpha=@n14, ConstraintWeights_0=1) {
    alphaBoolBlend {blendOption=Linear}
    boneToModify {boneName=dyn_loinClothFront_01_l}
    constraintSetup [{targetBone {boneName=thigh_l}}]
    constraintWeights [1]
   }
   n21 = "Constraint - dyn_loinClothFront_02_r (P)"(ComponentPose=@n20, Alpha=@n14, ConstraintWeights_0=1) {
    alphaBoolBlend {blendOption=Linear}
    boneToModify {boneName=dyn_loinClothFront_02_r}
    constraintSetup [{targetBone {boneName=thigh_r}}]
    constraintWeights [1]
   }
   n22 = "Constraint - dyn_loinClothFront_02_c (PP)"!(ComponentPose=@n21, Alpha=@n14, ConstraintWeights_0=0.5, ConstraintWeights_1:float=0.5) -> (Pose) {
    alphaBoolBlend {blendOption=Linear}
    boneToModify {boneName=dyn_loinClothFront_02_c}
    constraintSetup [{targetBone {boneName=thigh_r}}, {targetBone {boneName=thigh_l}}]
    constraintWeights [1,1]
   }
   n23 = "Constraint - dyn_loinClothFront_02_l (P)"(ComponentPose=@n22, Alpha=@n14, ConstraintWeights_0=1) {
    alphaBoolBlend {blendOption=Linear}
    boneToModify {boneName=dyn_loinClothFront_02_l}
    constraintSetup [{targetBone {boneName=thigh_l}}]
    constraintWeights [1]
   }
   n24 = "Constraint - dyn_loinClothFront_03_r (P)"(ComponentPose=@n23, Alpha=@n14, ConstraintWeights_0=1) {
    alphaBoolBlend {blendOption=Linear}
    boneToModify {boneName=dyn_loinClothFront_03_r}
    constraintSetup [{targetBone {boneName=thigh_r}}]
    constraintWeights [1]
   }
   n25 = "Constraint - dyn_loinClothFront_03_c (PP)"!(ComponentPose=@n24, Alpha=@n14, ConstraintWeights_0=0.5, ConstraintWeights_1:float=0.5) -> (Pose) {
    alphaBoolBlend {blendOption=Linear}
    boneToModify {boneName=dyn_loinClothFront_03_c}
    constraintSetup [{targetBone {boneName=thigh_r}}, {targetBone {boneName=thigh_l}}]
    constraintWeights [1,1]
   }
   n26 = "Constraint - dyn_loinClothFront_03_l (P)"(ComponentPose=@n25, Alpha=@n14, ConstraintWeights_0=1) {
    alphaBoolBlend {blendOption=Linear}
    boneToModify {boneName=dyn_loinClothFront_03_l}
    constraintSetup [{targetBone {boneName=thigh_l}}]
    constraintWeights [1]
   }
   n27 = "Transform (Modify) Bone - Bone: dyn_loinClothFront_l_MSTR"!(ComponentPose=@n26, bAlphaBoolEnabled:bool=True, Translation="0,0,0", Rotation="0.000,0.000,-5", Scale="1,1,1") -> (Pose) {
    alphaInputType=Bool
    rotationMode=BMM_Additive
    scaleMode=BMM_Additive
    translationMode=BMM_Additive
    alphaBoolBlend {blendOption=Linear}
    boneToModify {boneName=dyn_loinClothFront_l_MSTR}
    rotation {roll=-5}
   }
   n28 = "Transform (Modify) Bone - Bone: dyn_loinClothFront_c_MSTR"!(ComponentPose=@n27, bAlphaBoolEnabled:bool=True, Translation="0,0,0", Rotation="0.000,0.000,-5", Scale="1,1,1") -> (Pose) {
    alphaInputType=Bool
    rotationMode=BMM_Additive
    scaleMode=BMM_Additive
    translationMode=BMM_Additive
    alphaBoolBlend {blendOption=Linear}
    boneToModify {boneName=dyn_loinClothFront_c_MSTR}
    rotation {roll=-5}
   }
   n29 = Control Rig(Source=Use cached pose 'ConstrainPass'()) {
    alphaBoolBlend {blendOption=Linear}
    controlRigAssetReference {blueprintRigClass=&M_MED_Raptor_Knight_Pads_CR_C}
   }
   n30 = Control Rig(Source=@n29) {
    alphaBoolBlend {blendOption=Linear}
    controlRigAssetReference {blueprintRigClass=&/M_MED_Raptor_Knight_CTRL_Rig.M_MED_Raptor_Knight_CTRL_Rig_C}
   }
   n31 = "Constraint - dyn_beltBuckle_MSTR (PPPPPPPP)"!(
    ComponentPose=Local To Component(LocalPose=@n30)
    Alpha=1
    ConstraintWeights_0=0.0326
    ConstraintWeights_1:float=0.2435
    ConstraintWeights_2:float=0.0662
    ConstraintWeights_3:float=0.0328
    ConstraintWeights_4:float=0.0033
    ConstraintWeights_5:float=0.0001
    ConstraintWeights_6:float=0.5885
    ConstraintWeights_7:float=0.033
   ) -> (Pose) {
    alphaBoolBlend {blendOption=Linear}
    boneToModify {boneName=dyn_beltBuckle_MSTR}
    constraintSetup [
     {targetBone {boneName=thigh_twist_01_r}}
     {targetBone {boneName=spine_01}}
     {targetBone {boneName=spine_02}}
     {targetBone {boneName=spine_03}}
     {targetBone {boneName=spine_04}}
     {targetBone {boneName=spine_05}}
     {targetBone {boneName=pelvis}}
     {targetBone {boneName=thigh_twist_01_l}}
    ]
    constraintWeights [1,1,1,1,1,1,1,1]
   }
   n32 = "Constraint - dyn_l_knee (PP)"!(ComponentPose=@n31, Alpha=1, ConstraintWeights_0=0.4, ConstraintWeights_1:float=0.6) -> (Pose) {
    alphaBoolBlend {blendOption=Linear}
    boneToModify {boneName=dyn_l_knee}
    constraintSetup [{targetBone {boneName=thigh_twist_01_l}}, {targetBone {boneName=calf_l}}]
    constraintWeights [1,1]
   }
   n33 = "Constraint - dyn_r_knee (PP)"!(ComponentPose=@n32, Alpha=1, ConstraintWeights_0=0.5, ConstraintWeights_1:float=0.5) -> (Pose) {
    alphaBoolBlend {blendOption=Linear}
    boneToModify {boneName=dyn_r_knee}
    constraintSetup [{targetBone {boneName=thigh_twist_01_r}}, {targetBone {boneName=calf_r}}]
    constraintWeights [1,1]
   }
   n34 = RigidBody(ComponentPose=@n33, Alpha=0.5, OverrideWorldGravity=Get GravityOverride()) {
    bClampLinearTranslationLimitToRefPose=true
    bOverrideWorldGravity=true
    overridePhysicsAsset=&/M_MED_Raptor_Knight_Crouch_Physics
    simulationSpace=BaseBoneSpace
    alphaBoolBlend {blendOption=Linear}
    baseBoneRef {boneName=dyn_root_simspace}
    componentAppliedLinearAccClamp=<2000 2000 3000>
    overrideWorldGravity {z=-980}
   }
   n35 = "Transform (Modify) Bone - Bone: dyn_l_knee"!(ComponentPose=@n34, Alpha:float=1, Translation="-3,-3,0", Rotation="0,0,0", Scale="1,1,1") -> (Pose) {
    rotationMode=BMM_Additive
    rotationSpace=BCS_BoneSpace
    scaleMode=BMM_Additive
    scaleSpace=BCS_BoneSpace
    translationMode=BMM_Additive
    translationSpace=BCS_BoneSpace
    alphaBoolBlend {blendOption=Linear}
    boneToModify {boneName=dyn_l_knee}
    translation=<-3 -3>
   }
   n36 = "Transform (Modify) Bone - Bone: dyn_r_knee"!(ComponentPose=@n35, Alpha:float=1, Translation="2,2.5,0", Rotation="0,0,0", Scale="1,1,1") -> (Pose) {
    rotationMode=BMM_Additive
    rotationSpace=BCS_BoneSpace
    scaleMode=BMM_Additive
    scaleSpace=BCS_BoneSpace
    translationMode=BMM_Additive
    translationSpace=BCS_BoneSpace
    alphaBoolBlend {blendOption=Linear}
    boneToModify {boneName=dyn_r_knee}
    translation=<2 2.5>
   }
   n37 = "Transform (Modify) Bone - Bone: dyn_r_kneeClothFix"(ComponentPose=@n36, Alpha=1, Translation="0,0,0", Rotation="0,0,0", Scale="1,1,0.5") {
    rotationMode=BMM_Additive
    rotationSpace=BCS_BoneSpace
    scaleMode=BMM_Additive
    scaleSpace=BCS_BoneSpace
    translationMode=BMM_Additive
    translationSpace=BCS_BoneSpace
    alphaBoolBlend {blendOption=Linear}
    boneToModify {boneName=dyn_r_kneeClothFix}
    scale {z=0.5}
   }
   n38 = "Transform (Modify) Bone - Bone: dyn_r_kneeStrapFix"(ComponentPose=@n37, Alpha=1, Translation="0,0,0", Rotation="0,-30,0", Scale="1,1,1") {
    rotationMode=BMM_Additive
    rotationSpace=BCS_BoneSpace
    scaleMode=BMM_Additive
    scaleSpace=BCS_BoneSpace
    translationMode=BMM_Additive
    translationSpace=BCS_BoneSpace
    alphaBoolBlend {blendOption=Linear}
    boneToModify {boneName=dyn_r_kneeStrapFix}
    rotation {yaw=-30}
   }
   n39 = "Transform (Modify) Bone - Bone: dyn_beltBuckle"(ComponentPose=@n38, Alpha=1, Translation="0,2,4", Rotation="0.000,0.000,20", Scale="1,1,1") {
    rotationMode=BMM_Additive
    scaleMode=BMM_Additive
    translationMode=BMM_Additive
    alphaBoolBlend {blendOption=Linear}
    boneToModify {boneName=dyn_beltBuckle}
    rotation {roll=20}
    translation {y=2, z=4}
   }
   n40 = "Transform (Modify) Bone - Bone: dyn_l_shoulderPad"!(
    ComponentPose=@n39
    bAlphaBoolEnabled:bool=Get bIsAimingRifle() -> (bIsAimingRifle:bool)
    Translation="-1.000,0.000,-2"
    Rotation="-10,5.000,-20"
    Scale="1,1,1"
   ) -> (Pose) {
    alphaInputType=Bool
    rotationMode=BMM_Additive
    rotationSpace=BCS_BoneSpace
    scaleMode=BMM_Additive
    scaleSpace=BCS_BoneSpace
    translationMode=BMM_Additive
    translationSpace=BCS_BoneSpace
    alphaBoolBlend {blendInTime=0.1, blendOption=Linear, blendOutTime=0.1}
    boneToModify {boneName=dyn_l_shoulderPad}
    rotation=<rot -10 5 -20>
    translation {x=-1, z=-2}
   }
   n41 = "Transform (Modify) Bone - Bone: dyn_l_shoulderPad"(ComponentPose=@n40, bAlphaBoolEnabled=True, Translation="-2,-2,0.5", Rotation="-10,0,-17", Scale="1,1,1") {
    alphaInputType=Bool
    rotationMode=BMM_Additive
    rotationSpace=BCS_BoneSpace
    scaleMode=BMM_Additive
    scaleSpace=BCS_BoneSpace
    translationMode=BMM_Additive
    translationSpace=BCS_BoneSpace
    alphaBoolBlend {blendInTime=0.1, blendOption=Linear, blendOutTime=0.1}
    boneToModify {boneName=dyn_l_shoulderPad}
    rotation {pitch=-10, roll=-17}
    translation=<-2 -2 0.5>
   }
   n42 = "Transform (Modify) Bone - Bone: dyn_r_shoulderPad"!(ComponentPose=@n41, bAlphaBoolEnabled:bool=Get bIsAimingRifle(), Translation="0,0,0", Rotation="0,0,30", Scale="1,1,1") -> (Pose) {
    alphaInputType=Bool
    rotationMode=BMM_Additive
    rotationSpace=BCS_BoneSpace
    scaleMode=BMM_Additive
    scaleSpace=BCS_BoneSpace
    translationMode=BMM_Additive
    translationSpace=BCS_BoneSpace
    alphaBoolBlend {blendInTime=0.1, blendOption=Linear, blendOutTime=0.1}
    boneToModify {boneName=dyn_r_shoulderPad}
    rotation {roll=30}
   }
   n43 = "Transform (Modify) Bone - Bone: dyn_r_shoulderPad"(ComponentPose=@n42, bAlphaBoolEnabled=True, Translation="0,0,0", Rotation="0,0,30", Scale="1,1,1") {
    alphaInputType=Bool
    rotationMode=BMM_Additive
    rotationSpace=BCS_BoneSpace
    scaleMode=BMM_Additive
    scaleSpace=BCS_BoneSpace
    translationMode=BMM_Additive
    translationSpace=BCS_BoneSpace
    alphaBoolBlend {blendInTime=0.1, blendOption=Linear, blendOutTime=0.1}
    boneToModify {boneName=dyn_r_shoulderPad}
    rotation {roll=30}
   }
   n44 = "Transform (Modify) Bone - Bone: dyn_sword_01"(ComponentPose=@n43, bAlphaBoolEnabled=True, Translation="0,0,0", Rotation="0.000,0.000,-20", Scale="1,1,1") {
    alphaInputType=Bool
    rotationMode=BMM_Additive
    rotationSpace=BCS_BoneSpace
    scaleMode=BMM_Additive
    scaleSpace=BCS_BoneSpace
    translationMode=BMM_Additive
    translationSpace=BCS_BoneSpace
    alphaBoolBlend {blendInTime=0.1, blendOption=Linear, blendOutTime=0.1}
    boneToModify {boneName=dyn_sword_01}
    rotation {roll=-20}
   }
   n45 = "Transform (Modify) Bone - Bone: dyn_loinClothFront_l_MSTR"!(ComponentPose=@n44, Alpha:float=1, Translation="0,6,2", Rotation="0,15,0", Scale="1,1,1") -> (Pose) {
    rotationMode=BMM_Additive
    scaleMode=BMM_Additive
    translationMode=BMM_Additive
    alphaBoolBlend {blendOption=Linear}
    boneToModify {boneName=dyn_loinClothFront_l_MSTR}
    rotation {yaw=15}
    translation {y=6, z=2}
   }
   n46 = "Transform (Modify) Bone - Bone: dyn_loinClothFront_c_MSTR"!(ComponentPose=@n45, Alpha:float=1, Translation="0,4,4", Rotation="0,0,0", Scale="1,1,1") -> (Pose) {
    rotationMode=BMM_Additive
    scaleMode=BMM_Additive
    translationMode=BMM_Additive
    alphaBoolBlend {blendOption=Linear}
    boneToModify {boneName=dyn_loinClothFront_c_MSTR}
    translation {y=4, z=4}
   }
   n47 = "Transform (Modify) Bone - Bone: dyn_loinClothFront_r_MSTR"(ComponentPose=@n46, Alpha=1, Translation="0,1,2", Rotation="0,0,0", Scale="1,1,1") {
    rotationMode=BMM_Additive
    scaleMode=BMM_Additive
    translationMode=BMM_Additive
    alphaBoolBlend {blendOption=Linear}
    boneToModify {boneName=dyn_loinClothFront_r_MSTR}
    translation {y=1, z=2}
   }
   n48 = "Transform (Modify) Bone - Bone: dyn_loinClothFront_01_l"(ComponentPose=@n47, Alpha=1, Translation="0,0,1", Rotation="0,0,0", Scale="1,1,1") {
    rotationMode=BMM_Additive
    scaleMode=BMM_Additive
    translationMode=BMM_Additive
    alphaBoolBlend {blendOption=Linear}
    boneToModify {boneName=dyn_loinClothFront_01_l}
    translation {z=1}
   }
   n49 = "Transform (Modify) Bone - Bone: dyn_loinClothFront_01_c"(ComponentPose=@n48, Alpha=1, Translation="0,0,0", Rotation="0,0,0", Scale="1,1,1") {
    rotationMode=BMM_Additive
    scaleMode=BMM_Additive
    translationMode=BMM_Additive
    alphaBoolBlend {blendOption=Linear}
    boneToModify {boneName=dyn_loinClothFront_01_c}
   }
   n50 = "Transform (Modify) Bone - Bone: dyn_loinClothFront_01_r"(ComponentPose=@n49, Alpha=1, Translation="0,0,0", Rotation="0,0,0", Scale="1,1,1") {
    rotationMode=BMM_Additive
    scaleMode=BMM_Additive
    translationMode=BMM_Additive
    alphaBoolBlend {blendOption=Linear}
    boneToModify {boneName=dyn_loinClothFront_01_r}
   }
   n51 = "Transform (Modify) Bone - Bone: dyn_loinClothFront_02_l"(ComponentPose=@n50, Alpha=1, Translation="0,0,1", Rotation="0,0,10", Scale="1,1,1") {
    rotationMode=BMM_Additive
    scaleMode=BMM_Additive
    translationMode=BMM_Additive
    alphaBoolBlend {blendOption=Linear}
    boneToModify {boneName=dyn_loinClothFront_02_l}
    rotation {roll=10}
    translation {z=1}
   }
   n52 = "Transform (Modify) Bone - Bone: dyn_loinClothFront_02_c"(ComponentPose=@n51, Alpha=1, Translation="0,0,0", Rotation="0,0,0", Scale="1,1,1") {
    rotationMode=BMM_Additive
    scaleMode=BMM_Additive
    translationMode=BMM_Additive
    alphaBoolBlend {blendOption=Linear}
    boneToModify {boneName=dyn_loinClothFront_02_c}
   }
   n53 = "Transform (Modify) Bone - Bone: dyn_loinClothFront_02_r"(ComponentPose=@n52, Alpha=1, Translation="0,0,-3", Rotation="0,0,0", Scale="1,1,1") {
    rotationMode=BMM_Additive
    scaleMode=BMM_Additive
    translationMode=BMM_Additive
    alphaBoolBlend {blendOption=Linear}
    boneToModify {boneName=dyn_loinClothFront_02_r}
    translation {z=-3}
   }
   n54 = "Transform (Modify) Bone - Bone: dyn_loinClothFront_03_l"(ComponentPose=@n53, Alpha=1, Translation="0,0,0", Rotation="0,0,0", Scale="1,1,1") {
    rotationMode=BMM_Additive
    scaleMode=BMM_Additive
    translationMode=BMM_Additive
    alphaBoolBlend {blendOption=Linear}
    boneToModify {boneName=dyn_loinClothFront_03_l}
   }
   n55 = "Transform (Modify) Bone - Bone: dyn_loinClothFront_03_c"(ComponentPose=@n54, Alpha=1, Translation="0,0,0", Rotation="0,0,0", Scale="1,1,1") {
    rotationMode=BMM_Additive
    scaleMode=BMM_Additive
    translationMode=BMM_Additive
    alphaBoolBlend {blendOption=Linear}
    boneToModify {boneName=dyn_loinClothFront_03_c}
   }
   n56 = "Transform (Modify) Bone - Bone: dyn_loinClothFront_03_r"(ComponentPose=@n55, Alpha=1, Translation="0,0,0", Rotation="0,0,0", Scale="1,1,1") {
    rotationMode=BMM_Additive
    scaleMode=BMM_Additive
    translationMode=BMM_Additive
    alphaBoolBlend {blendOption=Linear}
    boneToModify {boneName=dyn_loinClothFront_03_r}
   }
   n57 = Blend Poses by bool(
    BlendPose_0=Component To Local(ComponentPose=@n56)
    BlendPose_1=Component To Local(ComponentPose=@n28)
    BlendTime_1=0.25
    bActiveValue=Get bIsCrouching()
   ) {blendType=Linear}
   n58 = Get GravityOverride()
   n59 = RigidBody(ComponentPose=Local To Component(LocalPose=Use cached pose 'ConstrainPass'()), OverrideWorldGravity=@n58) {
    bClampLinearTranslationLimitToRefPose=true
    bOverrideWorldGravity=true
    overridePhysicsAsset=&M_MED_Raptor_Knight_Physics
    simulationSpace=BaseBoneSpace
    alphaBoolBlend {blendOption=Linear}
    baseBoneRef {boneName=pelvis}
    componentAppliedLinearAccClamp=<2000 2000 3000>
    componentLinearAccScale=<0.35 0.35 0.4>
    componentLinearVelScale=<0.1 0.1 0.25>
    overrideWorldGravity {z=-980}
   }
   n60 = RigidBody(ComponentPose=@n59, OverrideWorldGravity=@n58) {
    bClampLinearTranslationLimitToRefPose=true
    bOverrideWorldGravity=true
    overridePhysicsAsset=&/M_MED_Raptor_Knight_SwordCrouch_Physics1
    simulationSpace=BaseBoneSpace
    alphaBoolBlend {blendOption=Linear}
    baseBoneRef {boneName=dyn_root_simspace}
    componentAppliedLinearAccClamp=<2000 2000 3000>
    componentLinearAccScale=<0.8 0.8 0.8>
    componentLinearVelScale=<0.15 0.15 0.15>
    overrideWorldGravity {z=-980}
   }
   n61 = Control Rig(Source=Component To Local(ComponentPose=@n60), Alpha=0.75) {
    alphaBoolBlend {blendOption=Linear}
    controlRigAssetReference {blueprintRigClass=&M_MED_Raptor_Knight_CTRL_Rig_C}
   }
   n62 = "Constraint - dyn_loinClothFront_r_MSTR (P)"(ComponentPose=Local To Component(LocalPose=@n61), Alpha=1, ConstraintWeights_0=1) {
    alphaBoolBlend {blendOption=Linear}
    boneToModify {boneName=dyn_loinClothFront_r_MSTR}
    constraintSetup [{targetBone {boneName=thigh_r}}]
    constraintWeights [1]
   }
   n63 = "Constraint - dyn_loinClothFront_c_MSTR (PP)"(ComponentPose=@n62, Alpha=1, ConstraintWeights_0=0.5, ConstraintWeights_1=0.5) {
    alphaBoolBlend {blendOption=Linear}
    boneToModify {boneName=dyn_loinClothFront_c_MSTR}
    constraintSetup [{targetBone {boneName=thigh_r}}, {targetBone {boneName=thigh_l}}]
    constraintWeights [1,1]
   }
   n64 = "Constraint - dyn_loinClothFront_l_MSTR (P)"(ComponentPose=@n63, Alpha=1, ConstraintWeights_0=1) {
    alphaBoolBlend {blendOption=Linear}
    boneToModify {boneName=dyn_loinClothFront_l_MSTR}
    constraintSetup [{targetBone {boneName=thigh_l}}]
    constraintWeights [1]
   }
   n65 = "Constraint - dyn_loinClothFront_01_r (P)"(ComponentPose=@n64, Alpha=1, ConstraintWeights_0=1) {
    alphaBoolBlend {blendOption=Linear}
    boneToModify {boneName=dyn_loinClothFront_01_r}
    constraintSetup [{targetBone {boneName=thigh_r}}]
    constraintWeights [1]
   }
   n66 = "Constraint - dyn_loinClothFront_01_c (PP)"(ComponentPose=@n65, Alpha=1, ConstraintWeights_0=0.5, ConstraintWeights_1=0.5) {
    alphaBoolBlend {blendOption=Linear}
    boneToModify {boneName=dyn_loinClothFront_01_c}
    constraintSetup [{targetBone {boneName=thigh_r}}, {targetBone {boneName=thigh_l}}]
    constraintWeights [1,1]
   }
   n67 = "Constraint - dyn_loinClothFront_01_l (P)"(ComponentPose=@n66, Alpha=1, ConstraintWeights_0=1) {
    alphaBoolBlend {blendOption=Linear}
    boneToModify {boneName=dyn_loinClothFront_01_l}
    constraintSetup [{targetBone {boneName=thigh_l}}]
    constraintWeights [1]
   }
   n68 = "Constraint - dyn_loinClothFront_02_r (P)"(ComponentPose=@n67, Alpha=1, ConstraintWeights_0=1) {
    alphaBoolBlend {blendOption=Linear}
    boneToModify {boneName=dyn_loinClothFront_02_r}
    constraintSetup [{targetBone {boneName=thigh_twist_01_r}}]
    constraintWeights [1]
   }
   n69 = "Constraint - dyn_loinClothFront_02_c (PP)"(ComponentPose=@n68, Alpha=1, ConstraintWeights_0=0.5, ConstraintWeights_1=0.5) {
    alphaBoolBlend {blendOption=Linear}
    boneToModify {boneName=dyn_loinClothFront_02_c}
    constraintSetup [{targetBone {boneName=thigh_r}}, {targetBone {boneName=thigh_l}}]
    constraintWeights [1,1]
   }
   n70 = "Constraint - dyn_loinClothFront_02_l (P)"(ComponentPose=@n69, Alpha=1, ConstraintWeights_0=1) {
    alphaBoolBlend {blendOption=Linear}
    boneToModify {boneName=dyn_loinClothFront_02_l}
    constraintSetup [{targetBone {boneName=thigh_l}}]
    constraintWeights [1]
   }
   n71 = "Constraint - dyn_loinClothFront_03_r (P)"(ComponentPose=@n70, Alpha=1, ConstraintWeights_0=1) {
    alphaBoolBlend {blendOption=Linear}
    boneToModify {boneName=dyn_loinClothFront_03_r}
    constraintSetup [{targetBone {boneName=thigh_r}}]
    constraintWeights [1]
   }
   n72 = "Constraint - dyn_loinClothFront_03_c (PP)"(ComponentPose=@n71, Alpha=1, ConstraintWeights_0=0.5, ConstraintWeights_1=0.5) {
    alphaBoolBlend {blendOption=Linear}
    boneToModify {boneName=dyn_loinClothFront_03_c}
    constraintSetup [{targetBone {boneName=thigh_r}}, {targetBone {boneName=thigh_l}}]
    constraintWeights [1,1]
   }
   n73 = "Constraint - dyn_loinClothFront_03_l (P)"(ComponentPose=@n72, Alpha=1, ConstraintWeights_0=1) {
    alphaBoolBlend {blendOption=Linear}
    boneToModify {boneName=dyn_loinClothFront_03_l}
    constraintSetup [{targetBone {boneName=thigh_l}}]
    constraintWeights [1]
   }
   n74 = Component To Local(ComponentPose=@n73)
   n75 = Get bIsPlayingEmote()
   n76 = Get IsSkyDivingAlpha() -> (IsSkyDivingAlpha:float)
   n77 = "Constraint - dyn_sword_MSTR (P)"(
    ComponentPose=Local To Component(LocalPose=Blend Poses by bool(BlendPose_0=@n74, BlendPose_1=@n57, BlendTime_1=0.3, bActiveValue=@n75) {blendType=Linear})
    Alpha=@n76
    ConstraintWeights_0=1
   ) {
    alphaBoolBlend {blendOption=Linear}
    boneToModify {boneName=dyn_sword_MSTR}
    constraintSetup [{targetBone {boneName=thigh_twist_01_l}}]
    constraintWeights [1]
   }
   n78 = "Constraint - dyn_sword_01 (P)"(ComponentPose=@n77, Alpha=@n76, ConstraintWeights_0=1) {
    alphaBoolBlend {blendOption=Linear}
    boneToModify {boneName=dyn_sword_01}
    constraintSetup [{targetBone {boneName=thigh_twist_01_l}}]
    constraintWeights [1]
   }
   n79 = "Constraint - dyn_sword_02 (P)"(ComponentPose=@n78, Alpha=@n76, ConstraintWeights_0=1) {
    alphaBoolBlend {blendOption=Linear}
    boneToModify {boneName=dyn_sword_02}
    constraintSetup [{targetBone {boneName=thigh_twist_01_l}}]
    constraintWeights [1]
   }
   n80 = "Constraint - dyn_sword_03 (P)"(ComponentPose=@n79, Alpha=@n76, ConstraintWeights_0=1) {
    alphaBoolBlend {blendOption=Linear}
    boneToModify {boneName=dyn_sword_03}
    constraintSetup [{targetBone {boneName=thigh_twist_01_l}}]
    constraintWeights [1]
   }
   n81 = "Transform (Modify) Bone - Bone: dyn_r_shoulderPad"!(ComponentPose=@n80, Alpha:float=1, Translation="0,0,0", Rotation="0,0,20", Scale="1,1,1") -> (Pose) {
    rotationMode=BMM_Additive
    rotationSpace=BCS_BoneSpace
    scaleMode=BMM_Additive
    scaleSpace=BCS_BoneSpace
    translationMode=BMM_Additive
    translationSpace=BCS_BoneSpace
    alphaBoolBlend {blendOption=Linear}
    boneToModify {boneName=dyn_r_shoulderPad}
    rotation {roll=20}
   }
   n82 = "Transform (Modify) Bone - Bone: dyn_l_shoulderPad"!(
    ComponentPose=@n81
    Alpha:float=Get IsAimingRifleAlpha() -> (IsAimingRifleAlpha:float)
    Translation="0,0,2"
    Rotation="0,0,20"
    Scale="1,1,1"
   ) -> (Pose) {
    rotationMode=BMM_Additive
    rotationSpace=BCS_BoneSpace
    scaleMode=BMM_Additive
    scaleSpace=BCS_BoneSpace
    translationMode=BMM_Additive
    translationSpace=BCS_BoneSpace
    alphaBoolBlend {blendOption=Linear}
    boneToModify {boneName=dyn_l_shoulderPad}
    rotation {roll=20}
    translation {z=2}
   }
   n83 = "Constraint - dyn_r_chestStrap_MSTR (PPPPPPPPPPP)"!(
    ComponentPose=@n82
    Alpha=1
    ConstraintWeights_0=0.0012
    ConstraintWeights_1:float=0.0039
    ConstraintWeights_2:float=0.0039
    ConstraintWeights_3:float=0.0035
    ConstraintWeights_4:float=0.063
    ConstraintWeights_5:float=0.213
    ConstraintWeights_6:float=0.434
    ConstraintWeights_7:float=0.2158
    ConstraintWeights_8:float=0.0597
    ConstraintWeights_9:float=0.0018
    ConstraintWeights_10:float=0.0003
   ) -> (Pose) {
    alphaBoolBlend {blendOption=Linear}
    boneToModify {boneName=dyn_r_chestStrap_MSTR}
    constraintSetup [
     {targetBone {boneName=upperarm_r}}
     {targetBone {boneName=upperarm_twist_01_r}}
     {targetBone {boneName=upperarm_twist_02_r}}
     {targetBone {boneName=spine_01}}
     {targetBone {boneName=spine_02}}
     {targetBone {boneName=spine_03}}
     {targetBone {boneName=spine_04}}
     {targetBone {boneName=spine_05}}
     {targetBone {boneName=dyn_r_clavicle}}
     {targetBone {boneName=neck_02}}
     {targetBone {boneName=neck_01}}
    ]
    constraintWeights [1,1,1,1,1,1,1,1,1,1,1]
   }
   n84 = "Constraint - dyn_l_chestStrap_MSTR (PPPPPPPPPPPP)"!(
    ComponentPose=@n83
    Alpha=1
    ConstraintWeights_0=0.453
    ConstraintWeights_1:float=0
    ConstraintWeights_2:float=0.0008
    ConstraintWeights_3:float=0
    ConstraintWeights_4:float=0.0061
    ConstraintWeights_5:float=0.0731
    ConstraintWeights_6:float=0.2415
    ConstraintWeights_7:float=0.162
    ConstraintWeights_8:float=0.0563
    ConstraintWeights_9:float=0.0031
    ConstraintWeights_10:float=0.0009
    ConstraintWeights_11:float=0.003
   ) -> (Pose) {
    alphaBoolBlend {blendOption=Linear}
    boneToModify {boneName=dyn_l_chestStrap_MSTR}
    constraintSetup [
     {targetBone {boneName=dyn_l_clavicle}}
     {targetBone {boneName=head}}
     {targetBone {boneName=upperarm_twist_02_l}}
     {targetBone {boneName=spine_01}}
     {targetBone {boneName=spine_02}}
     {targetBone {boneName=spine_03}}
     {targetBone {boneName=spine_04}}
     {targetBone {boneName=spine_05}}
     {targetBone {boneName=upperarm_twist_01_l}}
     {targetBone {boneName=neck_02}}
     {targetBone {boneName=neck_01}}
     {targetBone {boneName=upperarm_l}}
    ]
    constraintWeights [1,1,1,1,1,1,1,1,1,1,1,1]
   }
   n85 = "Constraint - dyn_zipper_MSTR (PPPPPP)"!(
    ComponentPose=@n84
    Alpha=1
    ConstraintWeights_0=0.0051
    ConstraintWeights_1:float=0.0162
    ConstraintWeights_2:float=0.1189
    ConstraintWeights_3:float=0.4199
    ConstraintWeights_4:float=0.4389
    ConstraintWeights_5:float=0.0009
   ) -> (Pose) {
    alphaBoolBlend {blendOption=Linear}
    boneToModify {boneName=dyn_zipper_MSTR}
    constraintSetup [
     {targetBone {boneName=clavicle_l}}
     {targetBone {boneName=spine_02}}
     {targetBone {boneName=spine_03}}
     {targetBone {boneName=spine_04}}
     {targetBone {boneName=spine_05}}
     {targetBone {boneName=neck_02}}
    ]
    constraintWeights [1,1,1,1,1,1]
   }
   n86 = "Constraint - dyn_bullet_a_MSTR (PPPPPPP)"!(
    ComponentPose=@n85
    Alpha=1
    ConstraintWeights_0=0.0045
    ConstraintWeights_1:float=0.0529
    ConstraintWeights_2:float=0.1639
    ConstraintWeights_3:float=0.4439
    ConstraintWeights_4:float=0.2386
    ConstraintWeights_5:float=0.0952
    ConstraintWeights_6:float=0.0011
   ) -> (Pose) {
    alphaBoolBlend {blendOption=Linear}
    boneToModify {boneName=dyn_bullet_a_MSTR}
    constraintSetup [
     {targetBone {boneName=spine_01}}
     {targetBone {boneName=spine_02}}
     {targetBone {boneName=spine_03}}
     {targetBone {boneName=spine_04}}
     {targetBone {boneName=spine_05}}
     {targetBone {boneName=clavicle_r}}
     {targetBone {boneName=pelvis}}
    ]
    constraintWeights [1,1,1,1,1,1,1]
   }
   n87 = "Constraint - dyn_bullet_b_MSTR (PPPPPPP)"!(
    ComponentPose=@n86
    Alpha=1
    ConstraintWeights_0=0.0174
    ConstraintWeights_1:float=0.1163
    ConstraintWeights_2:float=0.2197
    ConstraintWeights_3:float=0.4054
    ConstraintWeights_4:float=0.1808
    ConstraintWeights_5:float=0.0548
    ConstraintWeights_6:float=0.0056
   ) -> (Pose) {
    alphaBoolBlend {blendOption=Linear}
    boneToModify {boneName=dyn_bullet_b_MSTR}
    constraintSetup [
     {targetBone {boneName=spine_01}}
     {targetBone {boneName=spine_02}}
     {targetBone {boneName=spine_03}}
     {targetBone {boneName=spine_04}}
     {targetBone {boneName=spine_05}}
     {targetBone {boneName=clavicle_r}}
     {targetBone {boneName=pelvis}}
    ]
    constraintWeights [1,1,1,1,1,1,1]
   }
   n88 = "Constraint - dyn_bullet_c_MSTR (PPPPPPP)"!(
    ComponentPose=@n87
    Alpha=1
    ConstraintWeights_0=0.0366
    ConstraintWeights_1:float=0.1857
    ConstraintWeights_2:float=0.2772
    ConstraintWeights_3:float=0.3568
    ConstraintWeights_4:float=0.1166
    ConstraintWeights_5:float=0.0136
    ConstraintWeights_6:float=0.0136
   ) -> (Pose) {
    alphaBoolBlend {blendOption=Linear}
    boneToModify {boneName=dyn_bullet_c_MSTR}
    constraintSetup [
     {targetBone {boneName=spine_01}}
     {targetBone {boneName=spine_02}}
     {targetBone {boneName=spine_03}}
     {targetBone {boneName=spine_04}}
     {targetBone {boneName=spine_05}}
     {targetBone {boneName=clavicle_r}}
     {targetBone {boneName=pelvis}}
    ]
    constraintWeights [1,1,1,1,1,1,1]
   }
   n89 = "Constraint - dyn_sword_MSTR (PPP)"!(ComponentPose=@n88, Alpha=1, ConstraintWeights_0=0.0837, ConstraintWeights_1:float=0.1106, ConstraintWeights_2:float=0.8057) -> (Pose) {
    alphaBoolBlend {blendOption=Linear}
    boneToModify {boneName=dyn_sword_MSTR}
    constraintSetup [{targetBone {boneName=spine_01}}, {targetBone {boneName=thigh_twist_01_r}}, {targetBone {boneName=pelvis}}]
    constraintWeights [1,1,1]
   }
   n90 = Anim Dynamics - Bone: dyn_zipper(ComponentPose=@n89, Alpha=1) {
    angularDampingOverride=0.9999
    bGravityOverrideInSimSpace=true
    bOverrideAngularDamping=true
    simulationSpace=RootRelative
    alphaBoolBlend {blendOption=Linear}
    boundBone {boneName=dyn_zipper}
    physicsBodyDefinitions [{
      sphereCollisionRadius=0
      boundBone {boneName=dyn_zipper}
      boxExtents=<10 10 3>
      constraintSetup {angularLimitsMax=<2 75 45>, angularLimitsMin=<-2 -10 -45>, angularTarget=<0 0 0>}
      localJointOffset=<0 0 -2>
     }]
   }
   n91 = Anim Dynamics - Bone: dyn_r_chestStrap_01(ComponentPose=@n90, Alpha=1, ExternalForce="0,0,0") {
    angularDampingOverride=0.9999
    angularSpringConstant=200
    bAngularSpring=true
    bGravityOverrideInSimSpace=true
    bLinearSpring=true
    bOverrideAngularDamping=true
    bOverrideLinearDamping=true
    gravityScale=0.35
    linearDampingOverride=0.9999
    linearSpringConstant=300
    simulationSpace=RootRelative
    alphaBoolBlend {blendOption=Linear}
    boundBone {boneName=dyn_r_chestStrap_01}
    componentAppliedLinearAccClamp=<400 400 500>
    componentLinearAccScale {z=0.5}
    componentLinearVelScale {z=1}
    gravityOverride {z=-980}
    physicsBodyDefinitions [{
      sphereCollisionRadius=0
      boundBone {boneName=dyn_r_chestStrap_01}
      boxExtents=<5 20 20>
      constraintSetup {angularLimitsMax=<15 5 25>, angularLimitsMin=<-15 -5 -25>, angularTarget=<0 0 0>}
      localJointOffset=<2 0 0>
     }]
   }
   n92 = Anim Dynamics - Bone: dyn_l_chestStrap_01(ComponentPose=@n91, Alpha=1, ExternalForce="0,0,0") {
    angularDampingOverride=0.9999
    angularSpringConstant=200
    bAngularSpring=true
    bGravityOverrideInSimSpace=true
    bLinearSpring=true
    bOverrideAngularDamping=true
    bOverrideLinearDamping=true
    gravityScale=0.35
    linearDampingOverride=0.9999
    linearSpringConstant=300
    simulationSpace=RootRelative
    alphaBoolBlend {blendOption=Linear}
    boundBone {boneName=dyn_l_chestStrap_01}
    componentAppliedLinearAccClamp=<400 400 500>
    componentLinearAccScale {z=0.5}
    componentLinearVelScale {z=1}
    gravityOverride {z=-980}
    physicsBodyDefinitions [{
      sphereCollisionRadius=0
      boundBone {boneName=dyn_l_chestStrap_01}
      boxExtents=<5 20 20>
      constraintSetup {angularLimitsMax=<15 5 25>, angularLimitsMin=<-15 -5 -25>, angularTarget=<0 0 0>, linearAxesMax=<0 0 0.5>, linearAxesMin=<0 0 -0.5>}
      localJointOffset=<2 0 0>
     }]
   }
   n93 = Anim Dynamics - Bone: dyn_bullet_a_01(ComponentPose=@n92, Alpha=0.8, ExternalForce="0,0,0") {
    angularDampingOverride=0.9999
    angularSpringConstant=500
    bAngularSpring=true
    bGravityOverrideInSimSpace=true
    bLinearSpring=true
    bOverrideAngularDamping=true
    bOverrideLinearDamping=true
    gravityScale=0.35
    linearDampingOverride=0.9999
    linearSpringConstant=500
    simulationSpace=RootRelative
    alphaBoolBlend {blendOption=Linear}
    boundBone {boneName=dyn_bullet_a_01}
    componentAppliedLinearAccClamp=<400 400 500>
    componentLinearAccScale {z=0.5}
    componentLinearVelScale {z=1}
    gravityOverride {z=-980}
    physicsBodyDefinitions [{
      sphereCollisionRadius=0
      boundBone {boneName=dyn_bullet_a_01}
      boxExtents=<5 20 20>
      constraintSetup {angularLimitsMax=<5 5 20>, angularLimitsMin=<-5 -5 -20>, angularTarget=<0 0 0>}
      localJointOffset=<2 0 0>
     }]
   }
   n94 = Anim Dynamics - Bone: dyn_bullet_b_01(ComponentPose=@n93, Alpha=0.7, ExternalForce="0,0,0") {
    angularDampingOverride=0.9999
    angularSpringConstant=500
    bAngularSpring=true
    bGravityOverrideInSimSpace=true
    bLinearSpring=true
    bOverrideAngularDamping=true
    bOverrideLinearDamping=true
    gravityScale=0.35
    linearDampingOverride=0.9999
    linearSpringConstant=500
    simulationSpace=RootRelative
    alphaBoolBlend {blendOption=Linear}
    boundBone {boneName=dyn_bullet_b_01}
    componentAppliedLinearAccClamp=<400 400 500>
    componentLinearAccScale {z=0.5}
    componentLinearVelScale {z=1}
    gravityOverride {z=-980}
    physicsBodyDefinitions [{
      sphereCollisionRadius=0
      boundBone {boneName=dyn_bullet_b_01}
      boxExtents=<5 20 20>
      constraintSetup {angularLimitsMax=<5 5 20>, angularLimitsMin=<-5 -5 -20>, angularTarget=<0 0 0>}
      localJointOffset=<1 0 0>
     }]
   }
   n95 = Anim Dynamics - Bone: dyn_bullet_c_01(ComponentPose=@n94, Alpha=0.6, ExternalForce="0,0,0") {
    angularDampingOverride=0.9999
    angularSpringConstant=500
    bAngularSpring=true
    bGravityOverrideInSimSpace=true
    bLinearSpring=true
    bOverrideAngularDamping=true
    bOverrideLinearDamping=true
    gravityScale=0.35
    linearDampingOverride=0.9999
    linearSpringConstant=500
    simulationSpace=RootRelative
    alphaBoolBlend {blendOption=Linear}
    boundBone {boneName=dyn_bullet_c_01}
    componentAppliedLinearAccClamp=<400 400 500>
    componentLinearAccScale {z=0.5}
    componentLinearVelScale {z=1}
    gravityOverride {z=-980}
    physicsBodyDefinitions [{
      sphereCollisionRadius=0
      boundBone {boneName=dyn_bullet_c_01}
      boxExtents=<5 20 20>
      constraintSetup {angularLimitsMax=<5 5 20>, angularLimitsMin=<-5 -5 -20>, angularTarget=<0 0 0>}
      localJointOffset=<3 0 0>
     }]
   }
   n96 = Anim Dynamics - Bone: dyn_sword_01(ComponentPose=@n95, Alpha=0.5, ExternalForce="0,0,0") {
    angularDampingOverride=0.9999
    angularSpringConstant=100
    bAngularSpring=true
    bGravityOverrideInSimSpace=true
    bOverrideAngularDamping=true
    simulationSpace=RootRelative
    alphaBoolBlend {blendOption=Linear}
    boundBone {boneName=dyn_sword_01}
    physicsBodyDefinitions [{
      sphereCollisionRadius=0
      boundBone {boneName=dyn_sword_01}
      boxExtents=<8 15 15>
      constraintSetup {angularLimitsMax=<5 0 45>, angularLimitsMin=<-25 0 -45>, angularTarget=<0 0 0>}
      localJointOffset=<1 0 0>
     }]
   }
   n97 = Anim Dynamics - Bone: dyn_sword_02(ComponentPose=@n96, Alpha=0.5, ExternalForce="0,0,0") {
    angularDampingOverride=0.9999
    angularSpringConstant=100
    bAngularSpring=true
    bGravityOverrideInSimSpace=true
    bOverrideAngularDamping=true
    simulationSpace=RootRelative
    alphaBoolBlend {blendOption=Linear}
    boundBone {boneName=dyn_sword_02}
    physicsBodyDefinitions [{
      sphereCollisionRadius=0
      boundBone {boneName=dyn_sword_02}
      boxExtents=<8 15 15>
      constraintSetup {angularLimitsMax=<45 0 45>, angularLimitsMin=<-45 0 -45>, angularTarget=<0 0 0>}
      localJointOffset=<1 0 0>
     }]
   }
   n98 = Anim Dynamics - Bone: dyn_sword_03(ComponentPose=@n97, Alpha=0.5, ExternalForce="0,0,0") {
    angularDampingOverride=0.9999
    angularSpringConstant=100
    bAngularSpring=true
    bGravityOverrideInSimSpace=true
    bOverrideAngularDamping=true
    simulationSpace=RootRelative
    alphaBoolBlend {blendOption=Linear}
    boundBone {boneName=dyn_sword_03}
    physicsBodyDefinitions [{
      sphereCollisionRadius=0
      boundBone {boneName=dyn_sword_03}
      boxExtents=<8 15 15>
      constraintSetup {angularLimitsMax=<45 0 45>, angularLimitsMin=<-45 0 -45>, angularTarget=<0 0 0>}
      localJointOffset=<1 0 0>
     }]
   }
   n99 = Anim Dynamics - Bone: dyn_loinClothFront_02_l(ComponentPose=@n98, Alpha=1, ExternalForce="0,0,0") {
    angularDampingOverride=0.9999
    angularSpringConstant=350
    bAngularSpring=true
    bGravityOverrideInSimSpace=true
    bOverrideAngularDamping=true
    simulationSpace=RootRelative
    alphaBoolBlend {blendOption=Linear}
    boundBone {boneName=dyn_loinClothFront_02_l}
    physicsBodyDefinitions [{
      sphereCollisionRadius=0
      boundBone {boneName=dyn_loinClothFront_02_l}
      boxExtents=<8 15 15>
      constraintSetup {angularLimitsMax=<5 5 0>, angularLimitsMin=<-5 -5 -5>, angularTarget=<0 0 0>}
      localJointOffset=<2 0 0>
     }]
   }
   n100 = Anim Dynamics - Bone: dyn_loinClothFront_02_c(ComponentPose=@n99, Alpha=1, ExternalForce="0,0,0") {
    angularDampingOverride=0.9999
    angularSpringConstant=350
    bAngularSpring=true
    bGravityOverrideInSimSpace=true
    bOverrideAngularDamping=true
    simulationSpace=RootRelative
    alphaBoolBlend {blendOption=Linear}
    boundBone {boneName=dyn_loinClothFront_02_c}
    physicsBodyDefinitions [{
      sphereCollisionRadius=0
      boundBone {boneName=dyn_loinClothFront_02_c}
      boxExtents=<8 15 15>
      constraintSetup {angularLimitsMax=<20 20 0>, angularLimitsMin=<-20 -20 -45>, angularTarget=<0 0 0>}
      localJointOffset=<2 0 0>
     }]
   }
   n101 = Anim Dynamics - Bone: dyn_loinClothFront_02_r(ComponentPose=@n100, Alpha=1, ExternalForce="0,0,0") {
    angularDampingOverride=0.9999
    angularSpringConstant=350
    bAngularSpring=true
    bGravityOverrideInSimSpace=true
    bOverrideAngularDamping=true
    simulationSpace=RootRelative
    alphaBoolBlend {blendOption=Linear}
    boundBone {boneName=dyn_loinClothFront_02_r}
    physicsBodyDefinitions [{
      sphereCollisionRadius=0
      boundBone {boneName=dyn_loinClothFront_02_r}
      boxExtents=<8 15 15>
      constraintSetup {angularLimitsMax=<20 20 0>, angularLimitsMin=<-20 -20 -45>, angularTarget=<0 0 0>}
      localJointOffset=<2 0 0>
     }]
   }
   n102 = Anim Dynamics - Bone: dyn_loinClothFront_03_l(ComponentPose=@n101, Alpha=1, ExternalForce="0,0,0") {
    angularDampingOverride=0.9999
    angularSpringConstant=350
    bAngularSpring=true
    bGravityOverrideInSimSpace=true
    bOverrideAngularDamping=true
    simulationSpace=RootRelative
    alphaBoolBlend {blendOption=Linear}
    boundBone {boneName=dyn_loinClothFront_03_l}
    physicsBodyDefinitions [{
      sphereCollisionRadius=0
      boundBone {boneName=dyn_loinClothFront_03_l}
      boxExtents=<8 15 15>
      constraintSetup {angularLimitsMax=<5 5 0>, angularLimitsMin=<-5 -5 -5>, angularTarget=<0 0 0>}
      localJointOffset=<2 0 0>
     }]
   }
   n103 = Anim Dynamics - Bone: dyn_loinClothFront_03_c(ComponentPose=@n102, Alpha=1, ExternalForce="0,0,0") {
    angularDampingOverride=0.9999
    angularSpringConstant=350
    bAngularSpring=true
    bGravityOverrideInSimSpace=true
    bOverrideAngularDamping=true
    simulationSpace=RootRelative
    alphaBoolBlend {blendOption=Linear}
    boundBone {boneName=dyn_loinClothFront_03_c}
    physicsBodyDefinitions [{
      sphereCollisionRadius=0
      boundBone {boneName=dyn_loinClothFront_03_c}
      boxExtents=<8 15 15>
      constraintSetup {angularLimitsMax=<20 20 0>, angularLimitsMin=<-20 -20 -45>, angularTarget=<0 0 0>}
      localJointOffset=<2 0 0>
     }]
   }
   n104 = Anim Dynamics - Bone: dyn_loinClothFront_03_r(ComponentPose=@n103, Alpha=1, ExternalForce="0,0,0") {
    angularDampingOverride=0.9999
    angularSpringConstant=350
    bAngularSpring=true
    bGravityOverrideInSimSpace=true
    bOverrideAngularDamping=true
    simulationSpace=RootRelative
    alphaBoolBlend {blendOption=Linear}
    boundBone {boneName=dyn_loinClothFront_03_r}
    physicsBodyDefinitions [{
      sphereCollisionRadius=0
      boundBone {boneName=dyn_loinClothFront_03_r}
      boxExtents=<8 15 15>
      constraintSetup {angularLimitsMax=<20 20 0>, angularLimitsMin=<-20 -20 -45>, angularTarget=<0 0 0>}
      localJointOffset=<2 0 0>
     }]
   }
   n105 = Anim Dynamics - Bone: dyn_l_collar(ComponentPose=@n104, Alpha=1, ExternalForce="0,0,0") {
    angularDampingOverride=0.9999
    angularSpringConstant=200
    bAngularSpring=true
    bGravityOverrideInSimSpace=true
    bLinearSpring=true
    bOverrideAngularDamping=true
    bOverrideLinearDamping=true
    gravityScale=0.35
    linearDampingOverride=0.9999
    linearSpringConstant=300
    simulationSpace=RootRelative
    alphaBoolBlend {blendOption=Linear}
    boundBone {boneName=dyn_l_collar}
    componentAppliedLinearAccClamp=<400 400 500>
    componentLinearAccScale {z=0.5}
    componentLinearVelScale {z=1}
    gravityOverride {z=-980}
    physicsBodyDefinitions [{
      sphereCollisionRadius=0
      boundBone {boneName=dyn_l_collar}
      boxExtents=<10 20 20>
      constraintSetup {angularLimitsMax=<15 15 15>, angularLimitsMin=<-15 -15 -15>, angularTarget=<0 0 0>}
      localJointOffset=<-2 0 0>
     }]
   }
   n106 = Anim Dynamics - Bone: dyn_collar(ComponentPose=@n105, Alpha=1, ExternalForce="0,0,0") {
    angularDampingOverride=0.9999
    angularSpringConstant=200
    bAngularSpring=true
    bGravityOverrideInSimSpace=true
    bLinearSpring=true
    bOverrideAngularDamping=true
    bOverrideLinearDamping=true
    gravityScale=0.35
    linearDampingOverride=0.9999
    linearSpringConstant=300
    simulationSpace=RootRelative
    alphaBoolBlend {blendOption=Linear}
    boundBone {boneName=dyn_collar}
    componentAppliedLinearAccClamp=<400 400 500>
    componentLinearAccScale {z=0.5}
    componentLinearVelScale {z=1}
    gravityOverride {z=-980}
    physicsBodyDefinitions [{
      sphereCollisionRadius=0
      boundBone {boneName=dyn_collar}
      boxExtents=<10 20 20>
      constraintSetup {angularLimitsMax=<15 15 15>, angularLimitsMin=<-15 -15 -15>, angularTarget=<0 0 0>, linearAxesMax=<0 0 0.5>, linearAxesMin=<0 0 -0.5>}
      localJointOffset=<-2 0 0>
     }]
   }
   n107 = Anim Dynamics - Bone: dyn_r_collar(ComponentPose=@n106, Alpha=1, ExternalForce="0,0,0") {
    angularDampingOverride=0.9999
    angularSpringConstant=200
    bAngularSpring=true
    bGravityOverrideInSimSpace=true
    bLinearSpring=true
    bOverrideAngularDamping=true
    bOverrideLinearDamping=true
    gravityScale=0.35
    linearDampingOverride=0.9999
    linearSpringConstant=300
    simulationSpace=RootRelative
    alphaBoolBlend {blendOption=Linear}
    boundBone {boneName=dyn_r_collar}
    componentAppliedLinearAccClamp=<400 400 500>
    componentLinearAccScale {z=0.5}
    componentLinearVelScale {z=1}
    gravityOverride {z=-980}
    physicsBodyDefinitions [{
      sphereCollisionRadius=0
      boundBone {boneName=dyn_r_collar}
      boxExtents=<10 20 20>
      constraintSetup {angularLimitsMax=<15 15 15>, angularLimitsMin=<-15 -15 -15>, angularTarget=<0 0 0>, linearAxesMax=<0 0 0.5>, linearAxesMin=<0 0 -0.5>}
      localJointOffset=<-2 0 0>
     }]
   }
   n108 = Get bSwordPassEnabled_0() -> (bSwordPassEnabled_0:bool)
   n109 = "Transform (Modify) Bone - Bone: dyn_sword_01"(ComponentPose=@n107, bAlphaBoolEnabled=@n108, Translation="0,0,0", Rotation="20,-0.000,5", Scale="1,1,1") {
    alphaInputType=Bool
    rotationMode=BMM_Additive
    rotationSpace=BCS_BoneSpace
    scaleMode=BMM_Additive
    scaleSpace=BCS_BoneSpace
    translationMode=BMM_Additive
    translationSpace=BCS_BoneSpace
    alphaBoolBlend {blendInTime=0.3, blendOption=Linear, blendOutTime=0.1}
    boneToModify {boneName=dyn_sword_01}
    rotation {pitch=20, roll=5}
   }
   n110 = "Transform (Modify) Bone - Bone: dyn_sword_02"!(ComponentPose=@n109, bAlphaBoolEnabled:bool=@n108, Translation="0,0,0", Rotation="10,0,-5", Scale="1,1,1") -> (Pose) {
    alphaInputType=Bool
    rotationMode=BMM_Additive
    rotationSpace=BCS_BoneSpace
    scaleMode=BMM_Additive
    scaleSpace=BCS_BoneSpace
    translationMode=BMM_Additive
    translationSpace=BCS_BoneSpace
    alphaBoolBlend {blendInTime=0.35, blendOption=Linear, blendOutTime=0.15}
    boneToModify {boneName=dyn_sword_02}
    rotation {pitch=10, roll=-5}
   }
   n111 = "Transform (Modify) Bone - Bone: dyn_sword_03"!(ComponentPose=@n110, bAlphaBoolEnabled:bool=@n108, Translation="0,0,0", Rotation="-20,0,-5", Scale="1,1,1") -> (Pose) {
    alphaInputType=Bool
    rotationMode=BMM_Additive
    rotationSpace=BCS_BoneSpace
    scaleMode=BMM_Additive
    scaleSpace=BCS_BoneSpace
    translationMode=BMM_Additive
    translationSpace=BCS_BoneSpace
    alphaBoolBlend {blendInTime=0.5, blendOption=Linear, blendOutTime=0.2}
    boneToModify {boneName=dyn_sword_03}
    rotation {pitch=-20, roll=-5}
   }
   n112 = Get bSwordPassEnabled_1() -> (bSwordPassEnabled_1:bool)
   n113 = "Transform (Modify) Bone - Bone: dyn_sword_01"(ComponentPose=@n111, bAlphaBoolEnabled=@n112, Translation="0,0,0.5", Rotation="-5,-20,0", Scale="1,1,1") {
    alphaInputType=Bool
    rotationMode=BMM_Additive
    rotationSpace=BCS_BoneSpace
    scaleMode=BMM_Additive
    scaleSpace=BCS_BoneSpace
    translationMode=BMM_Additive
    translationSpace=BCS_BoneSpace
    alphaBoolBlend {blendInTime=0.3, blendOption=Linear, blendOutTime=0.1}
    boneToModify {boneName=dyn_sword_01}
    rotation {pitch=-5, yaw=-20}
    translation {z=0.5}
   }
   n114 = "Transform (Modify) Bone - Bone: dyn_sword_02"(ComponentPose=@n113, bAlphaBoolEnabled=@n112, Translation="0,0,1", Rotation="0,0,0", Scale="1,1,1") {
    alphaInputType=Bool
    rotationMode=BMM_Additive
    rotationSpace=BCS_BoneSpace
    scaleMode=BMM_Additive
    scaleSpace=BCS_BoneSpace
    translationMode=BMM_Additive
    translationSpace=BCS_BoneSpace
    alphaBoolBlend {blendInTime=0.35, blendOption=Linear, blendOutTime=0.15}
    boneToModify {boneName=dyn_sword_02}
    translation {z=1}
   }
   n115 = "Transform (Modify) Bone - Bone: dyn_sword_03"(ComponentPose=@n114, bAlphaBoolEnabled=@n112, Translation="0,0,1.5", Rotation="0,-20,0", Scale="1,1,1") {
    alphaInputType=Bool
    rotationMode=BMM_Additive
    rotationSpace=BCS_BoneSpace
    scaleMode=BMM_Additive
    scaleSpace=BCS_BoneSpace
    translationMode=BMM_Additive
    translationSpace=BCS_BoneSpace
    alphaBoolBlend {blendInTime=0.4, blendOption=Linear, blendOutTime=0.2}
    boneToModify {boneName=dyn_sword_03}
    rotation {yaw=-20}
    translation {z=1.5}
   }
   n116 = Get bSwordPassEnabled_2() -> (bSwordPassEnabled_2:bool)
   n117 = "Transform (Modify) Bone - Bone: dyn_sword_01"(ComponentPose=@n115, bAlphaBoolEnabled=@n116, Translation="0,0,0", Rotation="0,0,5", Scale="1,1,1") {
    alphaInputType=Bool
    rotationMode=BMM_Additive
    rotationSpace=BCS_BoneSpace
    scaleMode=BMM_Additive
    scaleSpace=BCS_BoneSpace
    translationMode=BMM_Additive
    translationSpace=BCS_BoneSpace
    alphaBoolBlend {blendInTime=0.3, blendOption=Linear, blendOutTime=0.1}
    boneToModify {boneName=dyn_sword_01}
    rotation {roll=5}
   }
   n118 = "Transform (Modify) Bone - Bone: dyn_sword_02"(ComponentPose=@n117, bAlphaBoolEnabled=@n116, Translation="0,0,0", Rotation="0,0,5", Scale="1,1,1") {
    alphaInputType=Bool
    rotationMode=BMM_Additive
    rotationSpace=BCS_BoneSpace
    scaleMode=BMM_Additive
    scaleSpace=BCS_BoneSpace
    translationMode=BMM_Additive
    translationSpace=BCS_BoneSpace
    alphaBoolBlend {blendInTime=0.35, blendOption=Linear, blendOutTime=0.15}
    boneToModify {boneName=dyn_sword_02}
    rotation {roll=5}
   }
   n119 = "Transform (Modify) Bone - Bone: dyn_sword_03"(ComponentPose=@n118, bAlphaBoolEnabled=@n116, Translation="0,0,0", Rotation="0,0,5", Scale="1,1,1") {
    alphaInputType=Bool
    rotationMode=BMM_Additive
    rotationSpace=BCS_BoneSpace
    scaleMode=BMM_Additive
    scaleSpace=BCS_BoneSpace
    translationMode=BMM_Additive
    translationSpace=BCS_BoneSpace
    alphaBoolBlend {blendInTime=0.4, blendOption=Linear, blendOutTime=0.2}
    boneToModify {boneName=dyn_sword_03}
    rotation {roll=5}
   }
   n120 = Get bSwordPassEnabled_3() -> (bSwordPassEnabled_3:bool)
   n121 = "Transform (Modify) Bone - Bone: dyn_sword_01"(ComponentPose=@n119, bAlphaBoolEnabled=@n120, Translation="0,0,0", Rotation="0.000,-5,0", Scale="1,1,1") {
    alphaInputType=Bool
    rotationMode=BMM_Additive
    rotationSpace=BCS_BoneSpace
    scaleMode=BMM_Additive
    scaleSpace=BCS_BoneSpace
    translationMode=BMM_Additive
    translationSpace=BCS_BoneSpace
    alphaBoolBlend {blendInTime=0.3, blendOption=Linear, blendOutTime=0.1}
    boneToModify {boneName=dyn_sword_01}
    rotation {yaw=-5}
   }
   n122 = "Transform (Modify) Bone - Bone: dyn_sword_02"(ComponentPose=@n121, bAlphaBoolEnabled=@n120, Translation="0,0,0", Rotation="0,-5,0", Scale="1,1,1") {
    alphaInputType=Bool
    rotationMode=BMM_Additive
    rotationSpace=BCS_BoneSpace
    scaleMode=BMM_Additive
    scaleSpace=BCS_BoneSpace
    translationMode=BMM_Additive
    translationSpace=BCS_BoneSpace
    alphaBoolBlend {blendInTime=0.35, blendOption=Linear, blendOutTime=0.15}
    boneToModify {boneName=dyn_sword_02}
    rotation {yaw=-5}
   }
   n123 = "Transform (Modify) Bone - Bone: dyn_sword_03"(ComponentPose=@n122, bAlphaBoolEnabled=@n120, Translation="0,0,0", Rotation="0,-5,0", Scale="1,1,1") {
    alphaInputType=Bool
    rotationMode=BMM_Additive
    rotationSpace=BCS_BoneSpace
    scaleMode=BMM_Additive
    scaleSpace=BCS_BoneSpace
    translationMode=BMM_Additive
    translationSpace=BCS_BoneSpace
    alphaBoolBlend {blendInTime=0.4, blendOption=Linear, blendOutTime=0.2}
    boneToModify {boneName=dyn_sword_03}
    rotation {yaw=-5}
   }
   n124 = "dyn_sword_01.rotateY = thigh_l.rotateY - Bone Driven Controller"!(ComponentPose:Component Space Pose Link Structure=@n123, Alpha:float=0.5) -> (Pose:Component Space Pose Link Structure) {
    is=title AnimGraphNode_BoneDrivenController
    bAffectTargetRotationY=true
    sourceComponent=RotationY
    alphaBoolBlend {blendOption=Linear}
    sourceBone {boneName=thigh_l}
    targetBone {boneName=dyn_sword_01}
   }
   n125 = Get IsFrontEndAlpha() -> (IsFrontEndAlpha:float)
   n126 = Get GravityOverride()
   n127 = RigidBody(ComponentPose=@n124, Alpha=@n125, OverrideWorldGravity=@n126) {
    bClampLinearTranslationLimitToRefPose=true
    bOverrideWorldGravity=true
    overridePhysicsAsset=&/M_MED_Raptor_Knight_FrontEnd_Physics
    simulationSpace=BaseBoneSpace
    alphaBoolBlend {blendOption=Linear}
    baseBoneRef {boneName=root}
    componentAppliedLinearAccClamp=<5000 5000 5000>
    componentLinearAccScale=<0.5 0.5 0.5>
    componentLinearVelScale=<0.5 0.5 0.5>
    overrideWorldGravity {z=-980}
   }
   n128 = RigidBody(ComponentPose=@n127, Alpha=@n125, OverrideWorldGravity=@n126) {
    bClampLinearTranslationLimitToRefPose=true
    bOverrideWorldGravity=true
    overridePhysicsAsset=&/M_MED_Raptor_Knight_Sword_Physics
    simulationSpace=BaseBoneSpace
    alphaBoolBlend {blendOption=Linear}
    baseBoneRef {boneName=root}
    componentAppliedLinearAccClamp=<5000 5000 5000>
    overrideWorldGravity {z=-980}
   }
   n129 = "Constraint - dyn_loinClothFront_r_MSTR (P)"!(ComponentPose=@n128, bAlphaBoolEnabled:bool=True, ConstraintWeights_0=1) -> (Pose) {
    alphaInputType=Bool
    alphaBoolBlend {blendOption=Linear}
    boneToModify {boneName=dyn_loinClothFront_r_MSTR}
    constraintSetup [{targetBone {boneName=thigh_r}}]
    constraintWeights [1]
   }
   n130 = "Constraint - dyn_loinClothFront_c_MSTR (PP)"!(ComponentPose=@n129, bAlphaBoolEnabled:bool=True, ConstraintWeights_0=0.5, ConstraintWeights_1=0.5) -> (Pose) {
    alphaInputType=Bool
    alphaBoolBlend {blendOption=Linear}
    boneToModify {boneName=dyn_loinClothFront_c_MSTR}
    constraintSetup [{targetBone {boneName=thigh_r}}, {targetBone {boneName=thigh_l}}]
    constraintWeights [1,1]
   }
   n131 = "Constraint - dyn_loinClothFront_l_MSTR (P)"!(ComponentPose=@n130, bAlphaBoolEnabled:bool=True, ConstraintWeights_0=1) -> (Pose) {
    alphaInputType=Bool
    alphaBoolBlend {blendOption=Linear}
    boneToModify {boneName=dyn_loinClothFront_l_MSTR}
    constraintSetup [{targetBone {boneName=thigh_l}}]
    constraintWeights [1]
   }
   n132 = "Constraint - dyn_loinClothFront_01_r (P)"!(ComponentPose=@n131, bAlphaBoolEnabled:bool=True, ConstraintWeights_0=1) -> (Pose) {
    alphaInputType=Bool
    alphaBoolBlend {blendOption=Linear}
    boneToModify {boneName=dyn_loinClothFront_01_r}
    constraintSetup [{targetBone {boneName=thigh_r}}]
    constraintWeights [1]
   }
   n133 = "Constraint - dyn_loinClothFront_01_c (PP)"!(ComponentPose=@n132, bAlphaBoolEnabled:bool=True, ConstraintWeights_0=0.5, ConstraintWeights_1=0.5) -> (Pose) {
    alphaInputType=Bool
    alphaBoolBlend {blendOption=Linear}
    boneToModify {boneName=dyn_loinClothFront_01_c}
    constraintSetup [{targetBone {boneName=thigh_r}}, {targetBone {boneName=thigh_l}}]
    constraintWeights [1,1]
   }
   n134 = "Constraint - dyn_loinClothFront_01_l (P)"!(ComponentPose=@n133, bAlphaBoolEnabled:bool=True, ConstraintWeights_0=1) -> (Pose) {
    alphaInputType=Bool
    alphaBoolBlend {blendOption=Linear}
    boneToModify {boneName=dyn_loinClothFront_01_l}
    constraintSetup [{targetBone {boneName=thigh_l}}]
    constraintWeights [1]
   }
   n135 = "Constraint - dyn_loinClothFront_02_r (P)"!(ComponentPose=@n134, bAlphaBoolEnabled:bool=True, ConstraintWeights_0=1) -> (Pose) {
    alphaInputType=Bool
    alphaBoolBlend {blendOption=Linear}
    boneToModify {boneName=dyn_loinClothFront_02_r}
    constraintSetup [{targetBone {boneName=thigh_r}}]
    constraintWeights [1]
   }
   n136 = "Constraint - dyn_loinClothFront_02_c (PP)"!(ComponentPose=@n135, bAlphaBoolEnabled:bool=True, ConstraintWeights_0=0.5, ConstraintWeights_1=0.5) -> (Pose) {
    alphaInputType=Bool
    alphaBoolBlend {blendOption=Linear}
    boneToModify {boneName=dyn_loinClothFront_02_c}
    constraintSetup [{targetBone {boneName=thigh_r}}, {targetBone {boneName=thigh_l}}]
    constraintWeights [1,1]
   }
   n137 = "Constraint - dyn_loinClothFront_02_l (P)"!(ComponentPose=@n136, bAlphaBoolEnabled:bool=True, ConstraintWeights_0=1) -> (Pose) {
    alphaInputType=Bool
    alphaBoolBlend {blendOption=Linear}
    boneToModify {boneName=dyn_loinClothFront_02_l}
    constraintSetup [{targetBone {boneName=thigh_l}}]
    constraintWeights [1]
   }
   n138 = "Constraint - dyn_loinClothFront_03_r (P)"!(ComponentPose=@n137, bAlphaBoolEnabled:bool=True, ConstraintWeights_0=1) -> (Pose) {
    alphaInputType=Bool
    alphaBoolBlend {blendOption=Linear}
    boneToModify {boneName=dyn_loinClothFront_03_r}
    constraintSetup [{targetBone {boneName=thigh_r}}]
    constraintWeights [1]
   }
   n139 = "Constraint - dyn_loinClothFront_03_c (PP)"!(ComponentPose=@n138, bAlphaBoolEnabled:bool=True, ConstraintWeights_0=0.5, ConstraintWeights_1=0.5) -> (Pose) {
    alphaInputType=Bool
    alphaBoolBlend {blendOption=Linear}
    boneToModify {boneName=dyn_loinClothFront_03_c}
    constraintSetup [{targetBone {boneName=thigh_r}}, {targetBone {boneName=thigh_l}}]
    constraintWeights [1,1]
   }
   n140 = "Constraint - dyn_loinClothFront_03_l (P)"!(ComponentPose=@n139, bAlphaBoolEnabled:bool=True, ConstraintWeights_0=1) -> (Pose) {
    alphaInputType=Bool
    alphaBoolBlend {blendOption=Linear}
    boneToModify {boneName=dyn_loinClothFront_03_l}
    constraintSetup [{targetBone {boneName=thigh_l}}]
    constraintWeights [1]
   }
   Output Pose(Result=Component To Local(ComponentPose=@n140)) {layerGroup=None}
   n141 = "Constraint - dyn_beltBuckle_MSTR (PPPPPPPP)"(
    ComponentPose=Local To Component(LocalPose=Use cached pose 'InitPass'())
    Alpha=1
    ConstraintWeights_0=0.0326
    ConstraintWeights_1=0.2435
    ConstraintWeights_2=0.0662
    ConstraintWeights_3=0.0328
    ConstraintWeights_4=0.0033
    ConstraintWeights_5=0.0001
    ConstraintWeights_6=0.5885
    ConstraintWeights_7=0.033
   ) {
    alphaBoolBlend {blendOption=Linear}
    boneToModify {boneName=dyn_beltBuckle_MSTR}
    constraintSetup [
     {targetBone {boneName=thigh_twist_01_r}}
     {targetBone {boneName=spine_01}}
     {targetBone {boneName=spine_02}}
     {targetBone {boneName=spine_03}}
     {targetBone {boneName=spine_04}}
     {targetBone {boneName=spine_05}}
     {targetBone {boneName=pelvis}}
     {targetBone {boneName=thigh_twist_01_l}}
    ]
    constraintWeights [1,1,1,1,1,1,1,1]
   }
   n142 = "Constraint - dyn_loinClothFront_02_l (R)"(ComponentPose=@n141, Alpha=1, ConstraintWeights_0=1) {
    alphaBoolBlend {blendOption=Linear}
    boneToModify {boneName=dyn_loinClothFront_02_l}
    constraintSetup [{transformType=Rotation, targetBone {boneName=thigh_l}}]
    constraintWeights [1]
   }
   n143 = "Constraint - dyn_loinClothFront_02_r (R)"(ComponentPose=@n142, Alpha=1, ConstraintWeights_0=1) {
    alphaBoolBlend {blendOption=Linear}
    boneToModify {boneName=dyn_loinClothFront_02_r}
    constraintSetup [{transformType=Rotation, targetBone {boneName=thigh_r}}]
    constraintWeights [1]
   }
   n144 = "Constraint - dyn_l_knee (PP)"(ComponentPose=@n143, Alpha=1, ConstraintWeights_0=0.4, ConstraintWeights_1=0.6) {
    alphaBoolBlend {blendOption=Linear}
    boneToModify {boneName=dyn_l_knee}
    constraintSetup [{targetBone {boneName=thigh_twist_01_l}}, {targetBone {boneName=calf_l}}]
    constraintWeights [1,1]
   }
   n145 = "Constraint - dyn_r_knee (PP)"(ComponentPose=@n144, Alpha=1, ConstraintWeights_0=0.5, ConstraintWeights_1=0.5) {
    alphaBoolBlend {blendOption=Linear}
    boneToModify {boneName=dyn_r_knee}
    constraintSetup [{targetBone {boneName=thigh_twist_01_r}}, {targetBone {boneName=calf_r}}]
    constraintWeights [1,1]
   }
   n146 = "Transform (Modify) Bone - Bone: dyn_l_knee"(ComponentPose=@n145, Alpha=1, Translation="-3,-1,0", Rotation="0,0,0", Scale="1,1,1") {
    rotationMode=BMM_Additive
    rotationSpace=BCS_BoneSpace
    scaleMode=BMM_Additive
    scaleSpace=BCS_BoneSpace
    translationMode=BMM_Additive
    translationSpace=BCS_BoneSpace
    alphaBoolBlend {blendOption=Linear}
    boneToModify {boneName=dyn_l_knee}
    translation=<-3 -1>
   }
   n147 = "Transform (Modify) Bone - Bone: dyn_r_knee"(ComponentPose=@n146, Alpha=1, Translation="2,0,0", Rotation="0,0,0", Scale="1,1,1") {
    rotationMode=BMM_Additive
    rotationSpace=BCS_BoneSpace
    scaleMode=BMM_Additive
    scaleSpace=BCS_BoneSpace
    translationMode=BMM_Additive
    translationSpace=BCS_BoneSpace
    alphaBoolBlend {blendOption=Linear}
    boneToModify {boneName=dyn_r_knee}
    translation {x=2}
   }
   n148 = "Constraint - dyn_l_shoulderPad (P)"(ComponentPose=@n147, Alpha=1, ConstraintWeights_0=1) {
    alphaBoolBlend {blendOption=Linear}
    boneToModify {boneName=dyn_l_shoulderPad}
    constraintSetup [{targetBone {boneName=upperarm_twist_01_l}}]
    constraintWeights [1]
   }
   n149 = "Constraint - dyn_r_shoulderPad (P)"(ComponentPose=@n148, Alpha=1, ConstraintWeights_0=1) {
    alphaBoolBlend {blendOption=Linear}
    boneToModify {boneName=dyn_r_shoulderPad}
    constraintSetup [{targetBone {boneName=upperarm_twist_01_r}}]
    constraintWeights [1]
   }
   Save cached pose 'ConstrainPass'(Pose=Component To Local(ComponentPose=@n149)) {cacheName=ConstrainPass}
   n150 = "Transform (Modify) Bone - Bone: dyn_root_slope"!(
    ComponentPose=Local To Component(LocalPose=Copy Pose From Mesh() {bUseAttachedParent=true})
    bAlphaBoolEnabled:bool=NOT Boolean(A=Get bCanPlayCustomAnimations() -> (bCanPlayCustomAnimations:bool))
    Translation=Get SlopeRootTrans() -> (SlopeRootTrans:Vector)
    Rotation=Get SlopeFootRot() -> (SlopeFootRot:Rotator)
    Scale="1,1,1"
   ) -> (Pose) {
    alphaInputType=Bool
    rotationMode=BMM_Replace
    rotationSpace=BCS_WorldSpace
    translationMode=BMM_Replace
    translationSpace=BCS_WorldSpace
    alphaBoolBlend {blendOption=Linear}
    boneToModify {boneName=dyn_root_slope}
   }
   n151 = "Transform (Modify) Bone - Bone: dyn_root_simspace"(ComponentPose=@n150, Alpha=1, Translation="0,0,0", Rotation="0,0,0", Scale="1,1,1") {
    rotationMode=BMM_Replace
    rotationSpace=BCS_WorldSpace
    translationSpace=BCS_BoneSpace
    alphaBoolBlend {blendOption=Linear}
    alphaScaleBiasClamp {bInterpResult=true, interpSpeedDecreasing=2, interpSpeedIncreasing=2}
    boneToModify {boneName=dyn_root_simspace}
   }
   Save cached pose 'InitPass'(Pose=Component To Local(ComponentPose=@n151)) {cacheName=InitPass}
  ]}
 EventGraph {nodes [
   Event OnParachuteOpened()
   > n1 = Delay(Duration=1.5)
   ResetDynamics(execute=[@n1.then, @n3.then, @n4.then, @n5.then, @n6.then], InTeleportType=TeleportPhysics)
   > ResetDynamics(InTeleportType=ResetPhysics)
   Event BlueprintUpdateAnimation()
   > n2 = Gravity Override State Function DEPRECATED(CharacterPartAnimInstance=Self-Reference(), GameTimeInSeconds=GetGameTimeInSeconds(), GravityOverrideParams=Get Gravity Override Params())
   > Set GravityOverride(GravityOverride=@n2)
   n3 = Event OnEndSkydiving()
   n4 = Event OnBeginSkydiving()
   n5 = Event OnEndPlayingEmote() {is=title K2Node_Event}
   n6 = Event OnBeginPlayingEmote() {is=title K2Node_Event}
  ]}
]
variables [Gravity Override Params {type=GravityOverrideParamsStruct, default {skydive_GoingUp {gravityOverride=<0 0 0>}}}]
