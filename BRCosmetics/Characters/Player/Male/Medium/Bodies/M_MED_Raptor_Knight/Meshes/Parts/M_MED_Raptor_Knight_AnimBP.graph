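# Post-process style cosmetic AnimBP for the Raptor Knight medium male body part, following the attached parent mesh through Copy Pose From Mesh.
# When bIsInFrontEnd is true, three chained additive Modify Bone nodes adjust the right ear, jaw and nose bridge.
# Each update maps head-to-neck_02 yaw delta from ref pose into clamped LookUp and LookDown values, which the AnimGraph here does not read.
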AnimBlueprint /BRCosmetics/Characters/Player/Male/Medium/Bodies/M_MED_Raptor_Knight/Meshes/Parts/M_MED_Raptor_Knight_AnimBP
target_skeleton=&/M_MED_Raptor_Knight_Skeleton
graphs [
 AnimGraph {nodes [
   n1 = Get bIsInFrontEnd()
   n2 = "Transform (Modify) Bone - Bone: dyn_r_ear"!(
    ComponentPose=Local To Component(LocalPose=Copy Pose From Mesh() {bUseAttachedParent=true})
    bAlphaBoolEnabled:bool=@n1
    Translation="-0.05,0,0"
    Rotation="0,0,0"
    Scale="1,1,1"
   ) -> (Pose) {
    alphaInputType=Bool
    rotationMode=BMM_Additive
    scaleMode=BMM_Additive
    translationMode=BMM_Additive
    alphaBoolBlend {blendOption=Linear}
    boneToModify {boneName=dyn_r_ear}
    translation {x=-0.05}
   }
   n3 = "Transform (Modify) Bone - Bone: C_jaw"!(ComponentPose=@n2, bAlphaBoolEnabled:bool=@n1, Translation="0,0,0", Rotation="0,0,0", Scale="1.05,1.05,1.05") -> (Pose) {
    alphaInputType=Bool
    rotationMode=BMM_Additive
    scaleMode=BMM_Additive
    translationMode=BMM_Additive
    alphaBoolBlend {blendOption=Linear}
    boneToModify {boneName=C_jaw}
    scale=<1.05 1.05 1.05>
   }
   n4 = "Transform (Modify) Bone - Bone: C_nose_bridge"!(ComponentPose=@n3, bAlphaBoolEnabled:bool=@n1, Translation="0,0.5,0", Rotation="0,0,0", Scale="1.01,1.01,1.01") -> (Pose) {
    alphaInputType=Bool
    rotationMode=BMM_Additive
    scaleMode=BMM_Additive
    translationMode=BMM_Additive
    alphaBoolBlend {blendOption=Linear}
    boneToModify {boneName=C_nose_bridge}
    scale=<1.01 1.01 1.01>
    translation {y=0.5}
   }
   Output Pose(Result=Component To Local(ComponentPose=@n4)) {layerGroup=None}
  ]}
 EventGraph {nodes [
   Event BlueprintUpdateAnimation()
   n1 = GetOwnerSkeletalMeshComponent()
   n2 = BreakRotator(InRot=BreakTransform(InTransform=GetDeltaTransformFromRefPose(self=@n1, BoneName=head, BaseName=neck_02)).Rotation)
   > Set LookUp(LookUp:double=MapRangeClamped(Value=@n2.Yaw, InRangeA=0, InRangeB=-15)) -> (Output_Get:double)
   > Set LookDown(LookDown:double=MapRangeClamped(Value=@n2.Yaw, InRangeA=0, InRangeB=15)) -> (Output_Get:double)
  ]}
]
variables [
 LookDown {type=number}
 LookUp {type=number}
]
